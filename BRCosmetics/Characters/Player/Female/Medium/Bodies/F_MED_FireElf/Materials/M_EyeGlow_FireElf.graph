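# Translucent unlit material for the FireElf character's glowing eye effect, built from scrolling smoke noise.
# A tiled smoke texture pans over time, offset by vertex color red, and its green channel blends two HDR colors.
# That color feeds HDR and FX emissive adjustment functions.
# Opacity multiplies the noise by an inverted sphere mask and an alpha scalar, with depth fade skipped on mobile.
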Material /BRCosmetics/Characters/Player/Female/Medium/Bodies/F_MED_FireElf/Materials/M_EyeGlow_FireElf
graph {
 properties {
  blendMode=BLEND_Translucent
  pixelDepthOffsetMode=PDOM_Legacy
  previewMesh=/BRCosmetics/Characters/Player/Female/Medium/Bodies/F_MED_FireElf/Meshes/Parts/S_FX_FireElf_EyeGlow
  shadingModel=MSM_Unlit
  usedShadingModels=Unlit
 }
 nodes [
  n1 = Multiply(
   A=TextureCoordinate() {uTiling=0.5, vTiling=0.1}
   B=ComponentMask(None=VectorParameter() {parameterName=TextureTileUV, defaultValue=<rgba 1 1 0 0>}.RGB) {g=true, r=true}
  )
  n2 = VertexColor()
  n3 = Time()
  n4 = ScalarParameter() {defaultValue=1, parameterName=Speed}
  n5 = TextureSample(UVs=Panner(Coordinate=Add(A=@n1, B=@n2.R), Time=Multiply(A=@n3, B=@n4)) {speedX=-0.1, speedY=1}) {texture=&T_Inky_Smoke_Tile}
  n6 = LinearInterpolate(
   A=VectorParameter() {parameterName=Color, defaultValue=<rgba 10 0 0.1768 1>}.RGB
   B=VectorParameter() {parameterName=Color_High, defaultValue=<rgba 10 1.3123 0 1>}.RGB
   Alpha=@n5.G
  )
  MaterialFunctionCall(Emissive=MaterialFunctionCall(LDR Input=Clamp(None=@n6), HDR Tint=@n6, HDR Intensity Multiplier=ScalarParameter() {defaultValue=2, parameterName=HDR Intensity}) {materialFunction=&DeriveHDRfromLDR}.RemappedRange) {materialFunction=&MF_FX_EmissiveAdjustments}
  n7 = TextureCoordinate()
  n8 = Constant() {r=0.5}
  n9 = ScalarParameter() {defaultValue=1.25, parameterName=Radius}
  n10 = ScalarParameter() {defaultValue=0.15, parameterName=Hardness}
  n11 = Multiply(A=@n5.G, B=OneMinus(None=SphereMask(A=@n7, B=@n8, Radius=@n9, Hardness=@n10) {hardnessPercent=0}))
  n12 = DepthFade(Opacity=@n11, FadeDistance=ScalarParameter() {defaultValue=1, parameterName=DepthFade})
  Multiply(
   A=ShadingPathSwitch(Default=@n12, Deferred=@n12, Mobile=@n11) {desc=Migrated from FeatureLevelSwitch}
   B=ScalarParameter() {defaultValue=12, parameterName=Alpha}
  )
 ]
}
