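# Unlit translucent skeletal-mesh material for animated flames on the FireElf character's hands.
# Three panned noise samples combine with a gradient fire line and contrast shaping into a mask blending two HDR colours for emissive.
# The noise also scales vertex normals, added to a scrolling gradient texture, for distortion weighted by vertex red.
# Opacity combines that mask, a vertical gradient and vertex red/green, softened by depth fade.
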Material /BRCosmetics/Characters/Player/Female/Medium/Bodies/F_MED_FireElf/Materials/M_FireHands
graph {
 properties {
  bUsedWithMeshDeformer=true
  bUsedWithMorphTargets=true
  bUsedWithSkeletalMesh=true
  blendMode=BLEND_Translucent
  pixelDepthOffsetMode=PDOM_Legacy
  shadingModel=MSM_Unlit
  usedShadingModels=Unlit
 }
 nodes [
  n1 = VertexNormalWS()
  n2 = ComponentMask(None=VectorParameter() {group=UVScale, parameterName=UVScale, defaultValue=<rgba 0.5 1 0 0>}.RGB) {g=true, r=true}
  n3 = Multiply(A=TextureCoordinate() {vTiling=0.5}, B=@n2)
  n4 = ScalarParameter() {parameterName=UVOffset}
  n5 = TextureSample(UVs=Panner(Coordinate=Add(A=@n3, B=@n4)) {speedX=-0.05, speedY=1}) {texture=&T_NoisePack_01}
  n6 = Constant() {r=0.2}
  n7 = TextureCoordinate() {uTiling=4}
  n8 = Multiply(A=Time(), B=ScalarParameter() {defaultValue=2, parameterName=Speed})
  n9 = Panner(Coordinate=Add(A=Multiply(A=@n2, B=@n7), B=@n4), Time=@n8) {speedX=0.1, speedY=1}
  n10 = Add(A=@n4, B=Multiply(A=@n2, B=TextureCoordinate() {uTiling=6, vTiling=1.5}))
  n11 = Multiply(
   A=Add(
    A=MaterialFunctionCall(In=@n5.G, Contrast=@n6) {materialFunction=&CheapContrast}.Result
    B=TextureSample(UVs=@n9) {texture=&T_NoisePack_01}.G
   )
   B=TextureSample(UVs=Panner(Coordinate=@n10, Time=@n8) {speedX=-0.1, speedY=1}) {texture=&T_NoisePack_01}.G
  )
  n12 = VertexColor()
  n13 = LinearInterpolate(A=ScalarParameter() {defaultValue=1, parameterName=Distort}, B=ScalarParameter() {defaultValue=2, parameterName=Distort Edge}, Alpha=@n12.R)
  n14 = Panner(Coordinate=Add(A=@n4, B=TextureCoordinate()), Time=@n8) {speedY=0.25}
  Add(
   A=Multiply(A=Multiply(A=@n1, B=@n11), B=@n13)
   B=Multiply(A=TextureSample(UVs=@n14) {texture=/Game/Global/Textures/Default/Masks_and_Gradients/CylGradient}.RGB, B=@n13) {constB=35}
  )
  n15 = Add(
   A=@n11
   B=Power(
    Base=MaterialFunctionCall() {materialFunction=&LinearGradient}.VGradient
    Exp=ScalarParameter() {defaultValue=15, parameterName=FireLine}
   )
  )
  n16 = Clamp(None=MaterialFunctionCall(In=Power(Base=@n15) {constExponent=0.5}, Contrast=ScalarParameter() {defaultValue=0.5, parameterName=HardEdgeness}) {materialFunction=&CheapContrast}.Result)
  n17 = LinearInterpolate(
   A=VectorParameter() {parameterName=Color_A, defaultValue=<rgba 10 0 0.1768 1>}.RGB
   B=VectorParameter() {parameterName=Color_B, defaultValue=<rgba 10 1.3123 0 1>}.RGB
   Alpha=@n16
  )
  n18 = VectorParameter() {parameterName=HDR_Tint, defaultValue=<rgba 1 1 1 0>}
  MaterialFunctionCall(Emissive=MaterialFunctionCall(LDR Input=Clamp(None=@n17), HDR Tint=Multiply(A=@n17, B=@n18.RGB)) {materialFunction=&DeriveHDRfromLDR}.RemappedRange) {materialFunction=&MF_FX_EmissiveAdjustments}
  n19 = MaterialFunctionCall() {materialFunction=&LinearGradient}
  n20 = ScalarParameter() {defaultValue=1.5, parameterName=GradientContrast}
  n21 = Multiply(A=@n16, B=Multiply(A=Power(Base=@n19.VGradient, Exp=@n20), B=OneMinus(None=@n12.R)))
  n22 = LinearInterpolate(Alpha=@n12.G) {constB=0.5}
  Multiply(
   A=ScalarParameter() {defaultValue=1, parameterName=Alpha}
   B=DepthFade(Opacity=Add(A=@n21, B=@n22), FadeDistance=ScalarParameter() {defaultValue=5, parameterName=FadeDistance}) {fadeDistanceDefault=50}
  )
 ]
}
