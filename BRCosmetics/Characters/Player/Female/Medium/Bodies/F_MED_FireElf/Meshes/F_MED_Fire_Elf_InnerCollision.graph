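# Physics asset providing the "inner collision" shapes for the female medium Fire Elf player body, covering the lower body and outer-front garment bones.
# Capsules, with a few spheres, wrap the pelvis, legs, feet and paired dyn_outerFront/FrontMid bones.
# Constraints chain each leg from the pelvis and each garment pair from its thigh, mostly with free angular motion.
# The sides differ: thigh_r is shapeless, and thigh_twist_01_r carries its capsule and garment chains.
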PhysicsAsset /BRCosmetics/Characters/Player/Female/Medium/Bodies/F_MED_FireElf/Meshes/F_MED_Fire_Elf_InnerCollision
bodies {
 calf_l [
  {shapeType=Sphere, radius=8.375, center=<-2.4443 -2.4457 4.7882>}
  {shapeType=Capsule, length=32.8125, radius=6.5859, center=<-16.5741 -2.0991 0.9427>, rotation=<rot 89.029 177.9337 -179.9998>}
 ]
 calf_r [
  {shapeType=Sphere, radius=7.625, center=<-1.8245 0 -0.9691>}
  {shapeType=Capsule, length=33.0215, radius=7.3769, center=<17.2729 1.9628 -0.0002>, rotation=<rot 90 90 87.9947>}
 ]
 dyn_outerFrontMid_03_l [{
   shapeType=Capsule
   length=6.4424
   radius=4.8088
   shapeName=dyn_outerFrontMid_03_l_capsule
   center=<-0.1183 0.2827 0.5555>
   rotation=<rot 0 -89.7784 -179.7629>
  }]
 dyn_outerFrontMid_03_r [{
   shapeType=Capsule
   length=8.1588
   radius=4.8165
   shapeName=dyn_outerFrontMid_03_r_capsule
   center=<0.1632 -0.1996 -0.3028>
   rotation=<rot 0 -89.8754 0.0118>
  }]
 dyn_outerFrontMid_04_l [{
   shapeType=Capsule
   length=12.3949
   radius=4.9426
   shapeName=dyn_outerFrontMid_04_l_capsule
   center=<-0.9026 -0.7317 -0.6871>
   rotation=<rot 0 -92.8173 -90.44>
  }]
 dyn_outerFrontMid_04_r [{
   shapeType=Capsule
   length=11.6464
   radius=5.0416
   shapeName=dyn_outerFrontMid_04_r_capsule
   center=<0.0137 -0.4475 -1.1234>
   rotation=<rot 0 -90.1611 -86.5149>
  }]
 dyn_outerFront_04_l [{
   shapeType=Capsule
   length=7.7787
   radius=3.2513
   shapeName=dyn_outerFront_04_l_capsule
   center=<0.4309 -0.8158 0.488>
   rotation=<rot 6.3184 -79.944 93.007>
  }]
 dyn_outerFront_04_r [{
   shapeType=Capsule
   length=5.4453
   radius=4.9663
   shapeName=dyn_outerFront_04_r_capsule
   center=<-1.278 0.8245 -0.1558>
   rotation=<rot 6.9709 -90.0001 -89.2013>
  }]
 dyn_outerFront_05_l [{
   shapeType=Capsule
   length=6.6159
   radius=4.7631
   shapeName=dyn_outerFront_05_l_capsule
   center=<-3.5462 -0.4784 0.8969>
   rotation=<rot 11.3069 -90.0007 2.1577>
  }]
 dyn_outerFront_05_r [{
   shapeType=Capsule
   length=8.0957
   radius=4.6587
   shapeName=dyn_outerFront_05_r_capsule
   center=<2.1887 0.6562 0.3178>
   rotation=<rot 10.4146 -90 -17.2838>
  }]
 foot_l [
  {shapeType=Sphere, radius=8.375, center=<7.6998 0.3091 0>}
  {shapeType=Capsule, length=12.2685, radius=10.0052, center=<-1.1035 6.1782 -0.5764>, rotation=<rot 0.0729 -177.2311 -89.9418>}
 ]
 foot_r [{shapeType=Capsule, length=12.2685, radius=10.0052, center=<1.1035 -6.1782 -0.5764>, rotation=<rot 0.0729 2.7689 -89.9418>}]
 pelvis [{shapeType=Capsule, length=11.5409, radius=7.6791, center=<-4.8458 -0.0002 0>}]
 thigh_l [
  {shapeType=Capsule, length=35.5529, radius=6.8225, center=<-20.6115 -0.0002 3.1554>, rotation=<rot 90 0 -8.6249>}
  {shapeType=Capsule, length=8.5, radius=6.5, center=<-38.0734 -0.6025 4.5641>, rotation=<rot 90 0 0>}
 ]
 thigh_r []
 thigh_twist_01_r [{shapeType=Capsule, length=24.2893, radius=6.8361, center=<-5.6437 0 -1.3331>, rotation=<rot 90 0 0>}]
}
constraints [
 {bone1Name=thigh_l, bone2Name=pelvis, swing1LimitDegrees=45, swing2LimitDegrees=45, twistLimitDegrees=45}
 {bone1Name=calf_r, bone2Name=thigh_r, swing1LimitDegrees=45, swing2LimitDegrees=45, twistLimitDegrees=45}
 {bone1Name=thigh_twist_01_r, bone2Name=thigh_r, swing1LimitDegrees=45, swing2LimitDegrees=45, twistLimitDegrees=45}
 {
  bone1Name=calf_l
  bone2Name=thigh_l
  swing1LimitDegrees=45
  swing1Motion=Free
  swing2LimitDegrees=45
  swing2Motion=Free
  twistLimitDegrees=45
  twistMotion=Free
 }
 {
  bone1Name=foot_l
  bone2Name=calf_l
  swing1LimitDegrees=45
  swing1Motion=Free
  swing2LimitDegrees=45
  swing2Motion=Free
  twistLimitDegrees=45
  twistMotion=Free
 }
 {
  bone1Name=foot_r
  bone2Name=calf_r
  swing1LimitDegrees=45
  swing1Motion=Free
  swing2LimitDegrees=45
  swing2Motion=Free
  twistLimitDegrees=45
  twistMotion=Free
 }
 {
  bone1Name=dyn_outerFront_04_r
  bone2Name=thigh_twist_01_r
  swing1LimitDegrees=45
  swing1Motion=Free
  swing2LimitDegrees=45
  swing2Motion=Free
  twistLimitDegrees=45
  twistMotion=Free
 }
 {
  bone1Name=dyn_outerFront_05_r
  bone2Name=dyn_outerFront_04_r
  swing1LimitDegrees=45
  swing1Motion=Free
  swing2LimitDegrees=45
  swing2Motion=Free
  twistLimitDegrees=45
  twistMotion=Free
 }
 {
  bone1Name=dyn_outerFrontMid_03_r
  bone2Name=thigh_twist_01_r
  swing1LimitDegrees=45
  swing1Motion=Free
  swing2LimitDegrees=45
  swing2Motion=Free
  twistLimitDegrees=45
  twistMotion=Free
 }
 {
  bone1Name=dyn_outerFrontMid_04_r
  bone2Name=dyn_outerFrontMid_03_r
  swing1LimitDegrees=45
  swing1Motion=Free
  swing2LimitDegrees=45
  swing2Motion=Free
  twistLimitDegrees=45
  twistMotion=Free
 }
 {
  bone1Name=dyn_outerFrontMid_03_l
  bone2Name=thigh_l
  swing1LimitDegrees=45
  swing1Motion=Free
  swing2LimitDegrees=45
  swing2Motion=Free
  twistLimitDegrees=45
  twistMotion=Free
 }
 {
  bone1Name=dyn_outerFrontMid_04_l
  bone2Name=dyn_outerFrontMid_03_l
  swing1LimitDegrees=45
  swing1Motion=Free
  swing2LimitDegrees=45
  swing2Motion=Free
  twistLimitDegrees=45
  twistMotion=Free
 }
 {
  bone1Name=dyn_outerFront_04_l
  bone2Name=thigh_l
  swing1LimitDegrees=45
  swing1Motion=Free
  swing2LimitDegrees=45
  swing2Motion=Free
  twistLimitDegrees=45
  twistMotion=Free
 }
 {
  bone1Name=dyn_outerFront_05_l
  bone2Name=dyn_outerFront_04_l
  swing1LimitDegrees=45
  swing1Motion=Free
  swing2LimitDegrees=45
  swing2Motion=Free
  twistLimitDegrees=45
  twistMotion=Free
 }
 {
  bone1Name=thigh_r
  bone2Name=pelvis
  swing1LimitDegrees=45
  swing1Motion=Free
  swing2LimitDegrees=45
  swing2Motion=Free
  twistLimitDegrees=45
  twistMotion=Free
 }
]
