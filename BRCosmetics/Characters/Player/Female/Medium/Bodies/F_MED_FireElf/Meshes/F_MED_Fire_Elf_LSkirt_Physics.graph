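# Physics asset for the left skirt of the medium female Fire Elf body, with capsule bodies on the skirt bones, pelvis, legs and left foot.
# Four skirt chains (front, front-mid, back-mid, back) hang from tiny base capsules, with front bases constrained to the left thigh and back bases to the pelvis.
# Swing and twist limits shape each chain, while cross-constraints, mostly angularly free, tie neighboring chains into a connected sheet.
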PhysicsAsset /BRCosmetics/Characters/Player/Female/Medium/Bodies/F_MED_FireElf/Meshes/F_MED_Fire_Elf_LSkirt_Physics
bodies {
 calf_l [{shapeType=Capsule, length=28.3125, radius=13.5859, center=<-15.2779 -2.6113 -0.2437>, rotation=<rot 85.5982 177.9337 -180>}]
 calf_r [{shapeType=Capsule, length=32.7715, radius=7.6269, center=<17.2729 1.9628 -0.0002>, rotation=<rot 90 90 87.9947>}]
 dyn_outerBackBase_l [{shapeType=Capsule, length=0.505, radius=0.505, shapeName=dyn_outerBackBase_l_capsule, rotation=<rot 0 -90 0>}]
 dyn_outerBackMidBase_l [{shapeType=Capsule, length=0.505, radius=0.505, shapeName=dyn_outerBackMidBase_l_capsule, rotation=<rot 0 -90 0>}]
 dyn_outerBackMid_02_l [{
   shapeType=Capsule
   length=4.3396
   radius=3.2404
   shapeName=dyn_outerBackMid_02_l_capsule
   center=<-3.7205 -0.6352 -0.016>
   rotation=<rot 0.5709 -88.2029 -3.2387>
  }]
 dyn_outerBackMid_03_l [{
   shapeType=Capsule
   length=4.2768
   radius=3.9131
   shapeName=dyn_outerBackMid_03_l_capsule
   center=<-2.9574 -0.4894 -0.0601>
   rotation=<rot 0.0001 -89.9094 -179.0697>
  }]
 dyn_outerBackMid_04_l [{
   shapeType=Capsule
   length=4.9426
   radius=3.521
   shapeName=dyn_outerBackMid_04_l_capsule
   center=<-3.7339 -0.2975 -0.1097>
   rotation=<rot 0 -91.8351 -3.2676>
  }]
 dyn_outerBack_02_l [{
   shapeType=Capsule
   length=9.6187
   radius=4.1021
   shapeName=dyn_outerBack_02_l_capsule
   center=<-3.5128 -0.8845 -0.112>
   rotation=<rot 0.0001 -99.9521 5.0747>
  }]
 dyn_outerBack_03_l [{
   shapeType=Capsule
   length=7.836
   radius=4.3954
   shapeName=dyn_outerBack_03_l_capsule
   center=<-2.7905 0.224 0.0581>
   rotation=<rot 13.304 -89.9999 -2.3501>
  }]
 dyn_outerBack_04_l [{
   shapeType=Capsule
   length=7.4196
   radius=4.1888
   shapeName=dyn_outerBack_04_l_capsule
   center=<-3.307 0.1421 0.4719>
   rotation=<rot 0.0001 -98.4265 -5.8721>
  }]
 dyn_outerFrontBase_l [{shapeType=Capsule, length=0.505, radius=0.505, rotation=<rot 0 -90 0>}]
 dyn_outerFrontMidBase_l [{shapeType=Capsule, length=0.505, radius=0.505, rotation=<rot 0 -90 0>}]
 dyn_outerFrontMid_02_l [{
   shapeType=Capsule
   length=4.6572
   radius=5.044
   shapeName=dyn_outerFrontMid_02_l_capsule
   center=<-5.7677 -0.4587 -0.002>
   rotation=<rot 0 93.2057 2.4986>
  }]
 dyn_outerFrontMid_03_l [{
   shapeType=Capsule
   length=3.65
   radius=5.15
   shapeName=dyn_outerFrontMid_03_l_capsule
   center=<-7.1456 -0.4806 -0.176>
   rotation=<rot 0 -89.7784 0.237>
  }]
 dyn_outerFrontMid_04_l [{
   shapeType=Capsule
   length=3.55
   radius=4.45
   shapeName=dyn_outerFrontMid_04_l_capsule
   center=<-5.8623 -0.4876 -1.511>
   rotation=<rot 0.0001 -92.8173 -30.4399>
  }]
 dyn_outerFront_02_l [{
   shapeType=Capsule
   length=4.9427
   radius=4.4059
   shapeName=dyn_outerFront_02_l_capsule
   center=<-4.7978 -0.9914 0.0845>
   rotation=<rot -4.2589 -90.0001 1.4961>
  }]
 dyn_outerFront_03_l [{
   shapeType=Capsule
   length=4.1393
   radius=5.9111
   shapeName=dyn_outerFront_03_l_capsule
   center=<-4.9439 -0.0579 0.3771>
   rotation=<rot 0 89.2193 -4.0295>
  }]
 dyn_outerFront_04_l [{
   shapeType=Capsule
   length=5.5287
   radius=3.5013
   shapeName=dyn_outerFront_04_l_capsule
   center=<-3.6393 -0.9529 -0.0827>
   rotation=<rot 6.7441 -90 1.8599>
  }]
 dyn_outerFront_05_l [{
   shapeType=Capsule
   length=4.8659
   radius=3.5131
   shapeName=dyn_outerFront_05_l_capsule
   center=<-3.5103 -0.5745 0.4168>
   rotation=<rot 11.3069 -90.0001 -17.8427>
  }]
 foot_l [{shapeType=Capsule, length=12.2685, radius=10.0052, center=<-1.1035 6.1782 -0.5764>, rotation=<rot 0.0729 -177.2311 -89.9418>}]
 pelvis [{shapeType=Capsule, length=11.5409, radius=7.6791, center=<-4.8458 -0.0002 0>}]
 thigh_l [{shapeType=Capsule, length=19.53, radius=7.57, center=<-26.1546 -0.9067 0.8679>, rotation=<rot 83.8748 180 180>}]
 thigh_twist_01_r [{shapeType=Capsule, length=24.2893, radius=6.8361, center=<-5.6437 0 -1.3331>, rotation=<rot 90 0 0>}]
}
constraints [
 {bone1Name=thigh_l, bone2Name=pelvis, swing1LimitDegrees=45, swing2LimitDegrees=45, twistLimitDegrees=45}
 {bone1Name=calf_r, bone2Name=pelvis, swing1LimitDegrees=45, swing2LimitDegrees=45, twistLimitDegrees=45}
 {bone1Name=thigh_twist_01_r, bone2Name=pelvis, swing1LimitDegrees=45, swing2LimitDegrees=45, twistLimitDegrees=45}
 {
  bone1Name=calf_l
  bone2Name=thigh_l
  swing1LimitDegrees=45
  swing1Motion=Free
  swing2LimitDegrees=45
  swing2Motion=Free
  twistLimitDegrees=45
  twistMotion=Free
 }
 {
  bone1Name=foot_l
  bone2Name=calf_l
  swing1LimitDegrees=45
  swing1Motion=Free
  swing2LimitDegrees=45
  swing2Motion=Free
  twistLimitDegrees=45
  twistMotion=Free
 }
 {bone1Name=dyn_outerFrontMid_02_l, bone2Name=dyn_outerFrontMidBase_l, swing1LimitDegrees=50, swing2LimitDegrees=50, twistLimitDegrees=15}
 {bone1Name=dyn_outerFrontMid_03_l, bone2Name=dyn_outerFrontMid_02_l, swing1LimitDegrees=60, swing2LimitDegrees=60, twistLimitDegrees=15}
 {bone1Name=dyn_outerFrontMid_04_l, bone2Name=dyn_outerFrontMid_03_l, swing1LimitDegrees=60, swing2LimitDegrees=60, twistLimitDegrees=15}
 {bone1Name=dyn_outerFront_02_l, bone2Name=dyn_outerFrontBase_l, swing1LimitDegrees=50, swing2LimitDegrees=50, twistLimitDegrees=15}
 {bone1Name=dyn_outerFront_03_l, bone2Name=dyn_outerFront_02_l, swing1LimitDegrees=60, swing2LimitDegrees=60, twistLimitDegrees=20}
 {bone1Name=dyn_outerFront_04_l, bone2Name=dyn_outerFront_03_l, swing1LimitDegrees=60, swing2LimitDegrees=60, twistLimitDegrees=20}
 {bone1Name=dyn_outerFront_05_l, bone2Name=dyn_outerFront_04_l, swing1LimitDegrees=60, swing2LimitDegrees=60, twistLimitDegrees=20}
 {bone1Name=dyn_outerBack_02_l, bone2Name=dyn_outerBackBase_l, swing1LimitDegrees=30, swing2LimitDegrees=30, twistLimitDegrees=10}
 {bone1Name=dyn_outerBack_03_l, bone2Name=dyn_outerBack_02_l, swing1LimitDegrees=60, swing2LimitDegrees=60, twistLimitDegrees=15}
 {bone1Name=dyn_outerBack_04_l, bone2Name=dyn_outerBack_03_l, swing1LimitDegrees=60, swing2LimitDegrees=60, twistLimitDegrees=15}
 {bone1Name=dyn_outerBackMid_02_l, bone2Name=dyn_outerBackMidBase_l, swing1LimitDegrees=30, swing2LimitDegrees=30, twistLimitDegrees=10}
 {bone1Name=dyn_outerBackMid_03_l, bone2Name=dyn_outerBackMid_02_l, swing1LimitDegrees=60, swing2LimitDegrees=60, twistLimitDegrees=15}
 {bone1Name=dyn_outerBackMid_04_l, bone2Name=dyn_outerBackMid_03_l, swing1LimitDegrees=60, swing2LimitDegrees=60, twistLimitDegrees=15}
 {
  bone1Name=dyn_outerFrontMid_04_l
  bone2Name=dyn_outerFront_04_l
  swing1LimitDegrees=45
  swing1Motion=Free
  swing2LimitDegrees=45
  swing2Motion=Free
  twistLimitDegrees=45
  twistMotion=Free
 }
 {
  bone1Name=dyn_outerBack_03_l
  bone2Name=dyn_outerBackMid_03_l
  swing1LimitDegrees=45
  swing1Motion=Free
  swing2LimitDegrees=45
  swing2Motion=Free
  twistLimitDegrees=45
  twistMotion=Free
 }
 {
  bone1Name=dyn_outerFrontMid_04_l
  bone2Name=dyn_outerBackMid_04_l
  swing1LimitDegrees=45
  swing1Motion=Free
  swing2LimitDegrees=45
  swing2Motion=Free
  twistLimitDegrees=45
  twistMotion=Free
 }
 {
  bone1Name=dyn_outerBackMidBase_l
  bone2Name=pelvis
  swing1LimitDegrees=45
  swing1Motion=Free
  swing2LimitDegrees=45
  swing2Motion=Free
  twistLimitDegrees=45
  twistMotion=Free
 }
 {
  bone1Name=dyn_outerBackBase_l
  bone2Name=pelvis
  swing1LimitDegrees=45
  swing1Motion=Free
  swing2LimitDegrees=45
  swing2Motion=Free
  twistLimitDegrees=45
  twistMotion=Free
 }
 {bone1Name=dyn_outerFrontMid_03_l, bone2Name=dyn_outerFront_03_l, swing1LimitDegrees=60, swing2LimitDegrees=60, twistLimitDegrees=60}
 {
  bone1Name=dyn_outerFrontMidBase_l
  bone2Name=thigh_l
  swing1LimitDegrees=45
  swing1Motion=Free
  swing2LimitDegrees=45
  swing2Motion=Free
  twistLimitDegrees=45
  twistMotion=Free
 }
 {
  bone1Name=dyn_outerFrontBase_l
  bone2Name=thigh_l
  swing1LimitDegrees=45
  swing1Motion=Free
  swing2LimitDegrees=45
  swing2Motion=Free
  twistLimitDegrees=45
  twistMotion=Free
 }
]
