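# Physics asset for the female medium Fire Elf RSkirtB mesh, defining bodies and joints for outer skirt bone chains.
# Capsules and a sphere cover pelvis and legs, while capsule chains form front, front-mid, back-mid and back strands per side.
# Back chains hang from pelvis via base bodies; front chains follow thighs, through twist-bone bases on the right only.
# Cross-constraints tie neighboring strands, mostly with free swing; left and right setups are asymmetric.
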PhysicsAsset /BRCosmetics/Characters/Player/Female/Medium/Bodies/F_MED_FireElf/Meshes/F_MED_Fire_Elf_RSkirtB_Physics
bodies {
 calf_l [{shapeType=Capsule, length=30.3125, radius=10.0859, center=<-15.2779 -2.6113 -0.2437>, rotation=<rot 85.5982 177.9337 -180>}]
 calf_r [{shapeType=Capsule, length=29.6118, radius=10.2865, center=<15.2779 2.6113 -0.2437>, rotation=<rot 89.8264 -2.069 -178.8908>}]
 dyn_outerBackBase_l [{shapeType=Box, extentX=3.4707, extentY=3.2042, extentZ=4.9874, rotation=<rot 0 90 0>}]
 dyn_outerBackBase_r [{shapeType=Box, extentX=3.4707, extentY=3.2042, extentZ=4.9874, rotation=<rot 0 -90 0>}]
 dyn_outerBackMidBase_l [{shapeType=Box, extentX=1.6307, extentY=2.0762, extentZ=2.3037, rotation=<rot 0 90 0>}]
 dyn_outerBackMidBase_r [{shapeType=Box, extentX=1.6307, extentY=2.0762, extentZ=2.3037, rotation=<rot 0 -90 0>}]
 dyn_outerBackMid_02_l [{
   shapeType=Capsule
   length=4.3396
   radius=3.4904
   shapeName=dyn_outerBackMid_02_l_capsule
   center=<-4.2928 -0.5529 -0.1514>
   rotation=<rot 0 -98.1868 -3.2888>
  }]
 dyn_outerBackMid_02_r [{
   shapeType=Capsule
   length=5.5906
   radius=3
   shapeName=dyn_outerBackMid_02_r_capsule
   center=<3.7306 0.3189 0.601>
   rotation=<rot 0.5997 -88.0807 -3.4024>
  }]
 dyn_outerBackMid_03_l [{
   shapeType=Capsule
   length=3.7768
   radius=3.1631
   shapeName=dyn_outerBackMid_03_l_capsule
   center=<-3.3715 -0.4901 -0.0534>
   rotation=<rot 0 -89.9095 0.93>
  }]
 dyn_outerBackMid_03_r [{
   shapeType=Capsule
   length=4.3561
   radius=3
   shapeName=dyn_outerBackMid_03_r_capsule
   center=<3.4368 0.1503 -0.1108>
   rotation=<rot 0 -90.0665 2.613>
  }]
 dyn_outerBackMid_04_l [{
   shapeType=Capsule
   length=5.1926
   radius=5.021
   shapeName=dyn_outerBackMid_04_l_capsule
   center=<-3.849 -0.2938 0.3206>
   rotation=<rot 0 -91.8351 -3.2676>
  }]
 dyn_outerBackMid_04_r [{
   shapeType=Capsule
   length=4.1681
   radius=3
   shapeName=dyn_outerBackMid_04_r_capsule
   center=<4.8967 -0.1889 0.0065>
   rotation=<rot 0 -95.4461 -4.9493>
  }]
 dyn_outerBack_02_l [{
   shapeType=Capsule
   length=5.6187
   radius=3.3521
   shapeName=dyn_outerBack_02_l_capsule
   center=<-3.922 -0.7581 -0.0931>
   rotation=<rot 0.8543 -89.9882 -4.8508>
  }]
 dyn_outerBack_02_r [{
   shapeType=Capsule
   length=8.7587
   radius=2
   shapeName=dyn_outerBack_02_r_capsule
   center=<4.1568 1.237 -0.1063>
   rotation=<rot 0.0001 -99.1925 -9.8806>
  }]
 dyn_outerBack_03_l [{
   shapeType=Capsule
   length=3.836
   radius=2.8954
   shapeName=dyn_outerBack_03_l_capsule
   center=<-3.9032 0.1486 -0.467>
   rotation=<rot 3.3121 -89.5912 7.7092>
  }]
 dyn_outerBack_03_r [{
   shapeType=Capsule
   length=9.3326
   radius=4
   shapeName=dyn_outerBack_03_r_capsule
   center=<3.7579 -0.7535 -0.2371>
   rotation=<rot 0.6428 -89.036 -5.4621>
  }]
 dyn_outerBack_04_l [{
   shapeType=Capsule
   length=4.6696
   radius=4.4388
   shapeName=dyn_outerBack_04_l_capsule
   center=<-3.1019 0.1117 0.36>
   rotation=<rot 0 -98.4266 4.1278>
  }]
 dyn_outerBack_04_r [{
   shapeType=Capsule
   length=11.8895
   radius=4
   shapeName=dyn_outerBack_04_r_capsule
   center=<4.4762 -0.0742 -0.1679>
   rotation=<rot 0.0002 -101.1021 -15.7582>
  }]
 dyn_outerFrontBase_r [{shapeType=Capsule, length=0.505, radius=0.505, rotation=<rot 0 -90 0>}]
 dyn_outerFrontMidBase_r [{shapeType=Capsule, length=0.505, radius=0.505, rotation=<rot 0 -90 0>}]
 dyn_outerFrontMid_02_l [{
   shapeType=Capsule
   length=5.9072
   radius=3.544
   shapeName=dyn_outerFrontMid_02_l_capsule
   center=<-4.2982 -0.3764 0.1548>
   rotation=<rot 0 93.2057 2.4989>
  }]
 dyn_outerFrontMid_02_r [{
   shapeType=Capsule
   length=3.1757
   radius=4.0667
   shapeName=dyn_outerFrontMid_02_r_capsule
   center=<4.8859 -0.1849 1.5402>
   rotation=<rot 0 93.54 -86.6226>
  }]
 dyn_outerFrontMid_03_l [{
   shapeType=Capsule
   length=5.6924
   radius=4.0588
   shapeName=dyn_outerFrontMid_03_l_capsule
   center=<-4.3423 -0.4836 0.4273>
   rotation=<rot 0 -89.7784 -179.763>
  }]
 dyn_outerFrontMid_03_r [{
   shapeType=Capsule
   length=5.9088
   radius=4.0665
   shapeName=dyn_outerFrontMid_03_r_capsule
   center=<6.3788 -0.1861 1.5865>
   rotation=<rot 0 -89.8754 -89.9882>
  }]
 dyn_outerFrontMid_04_l [{
   shapeType=Capsule
   length=5.6449
   radius=4.1926
   shapeName=dyn_outerFrontMid_04_l_capsule
   center=<-5.5875 -0.5012 -1.1698>
   rotation=<rot 0.0001 -92.8173 -30.44>
  }]
 dyn_outerFrontMid_04_r [{
   shapeType=Capsule
   length=6.1464
   radius=3.0416
   shapeName=dyn_outerFrontMid_04_r_capsule
   center=<3.4287 -1.3122 1.0459>
   rotation=<rot 9.9813 -89.547 3.5388>
  }]
 dyn_outerFront_02_l [{
   shapeType=Capsule
   length=4.9427
   radius=4.4059
   shapeName=dyn_outerFront_02_l_capsule
   center=<-4.0197 0.9771 0.2107>
   rotation=<rot -4.2589 -90.0001 1.4961>
  }]
 dyn_outerFront_02_r [{
   shapeType=Capsule
   length=5.7143
   radius=3.6993
   shapeName=dyn_outerFront_02_r_capsule
   center=<6.4353 0.7184 -0.3419>
   rotation=<rot -2.1195 -90.0001 4.2424>
  }]
 dyn_outerFront_03_l [{
   shapeType=Capsule
   length=6.3893
   radius=4.1611
   shapeName=dyn_outerFront_03_l_capsule
   center=<-5.2147 0.8423 0.0176>
   rotation=<rot 0 89.2193 -4.0295>
  }]
 dyn_outerFront_03_r [{
   shapeType=Capsule
   length=6.6559
   radius=3.9332
   shapeName=dyn_outerFront_03_r_capsule
   center=<5.5899 0.7168 -0.3282>
   rotation=<rot 0 90.8082 -92.9189>
  }]
 dyn_outerFront_04_l [{
   shapeType=Capsule
   length=5.2787
   radius=3.2513
   shapeName=dyn_outerFront_04_l_capsule
   center=<-4.5466 -0.9495 -0.0534>
   rotation=<rot 6.7441 -90 1.8599>
  }]
 dyn_outerFront_04_r [{
   shapeType=Capsule
   length=7.4453
   radius=2.7163
   shapeName=dyn_outerFront_04_r_capsule
   center=<3.9287 0.8157 -0.2279>
   rotation=<rot 6.9709 -90 0.7986>
  }]
 dyn_outerFront_05_l [{
   shapeType=Capsule
   length=5.1159
   radius=3.5131
   shapeName=dyn_outerFront_05_l_capsule
   center=<-4.273 0.4159 0.1718>
   rotation=<rot 11.3069 -90.0001 -7.8427>
  }]
 dyn_outerFront_05_r [{
   shapeType=Capsule
   length=8.5957
   radius=3.1587
   shapeName=dyn_outerFront_05_r_capsule
   center=<4.0857 0.4284 -0.9218>
   rotation=<rot 10.4146 -90.0001 2.7163>
  }]
 foot_l [{shapeType=Capsule, length=12.2685, radius=10.0052, center=<-1.1035 6.1782 -0.5764>, rotation=<rot 0.0729 -177.2311 -89.9418>}]
 foot_r [{shapeType=Capsule, length=8.9502, radius=9.5779, center=<1.1148 -5.6791 0.5976>, rotation=<rot 0.1407 -174.1216 -89.7721>}]
 pelvis [{shapeType=Capsule, length=11.5409, radius=7.6791, center=<-4.8458 -0.0002 0>}]
 thigh_l [{shapeType=Capsule, length=25.9279, radius=8.5725, center=<-21.7918 -0.0002 1.5602>, rotation=<rot 83.887 180 180>}]
 thigh_r [
  {shapeType=Sphere, radius=11.099, center=<32.5772 2.1171 -4.7064>}
  {shapeType=Capsule, length=16.815, radius=9.9108, center=<26.8703 0.6477 -1.6069>, rotation=<rot 85.0886 -179.9998 0.0001>}
 ]
 thigh_twist_01_r [{shapeType=Capsule, length=24.2893, radius=6.8361, center=<-3.5096 0.0002 -1.3331>, rotation=<rot 90 0 0>}]
}
constraints [
 {bone1Name=thigh_l, bone2Name=pelvis, swing1LimitDegrees=45, swing2LimitDegrees=45, twistLimitDegrees=45}
 {bone1Name=thigh_twist_01_r, bone2Name=thigh_r, swing1LimitDegrees=45, swing2LimitDegrees=45, twistLimitDegrees=45}
 {
  bone1Name=calf_l
  bone2Name=thigh_l
  swing1LimitDegrees=45
  swing1Motion=Free
  swing2LimitDegrees=45
  swing2Motion=Free
  twistLimitDegrees=45
  twistMotion=Free
 }
 {
  bone1Name=foot_l
  bone2Name=calf_l
  swing1LimitDegrees=45
  swing1Motion=Free
  swing2LimitDegrees=45
  swing2Motion=Free
  twistLimitDegrees=45
  twistMotion=Free
 }
 {bone1Name=dyn_outerFrontMid_02_l, bone2Name=thigh_l, swing1LimitDegrees=60, swing2LimitDegrees=60, twistLimitDegrees=15}
 {bone1Name=dyn_outerFrontMid_03_l, bone2Name=dyn_outerFrontMid_02_l, swing1LimitDegrees=60, swing2LimitDegrees=60, twistLimitDegrees=15}
 {bone1Name=dyn_outerFrontMid_04_l, bone2Name=dyn_outerFrontMid_03_l, swing1LimitDegrees=60, swing2LimitDegrees=60, twistLimitDegrees=15}
 {bone1Name=dyn_outerFront_02_l, bone2Name=thigh_l, swing1LimitDegrees=60, swing2LimitDegrees=60, twistLimitDegrees=15}
 {bone1Name=dyn_outerFront_03_l, bone2Name=dyn_outerFront_02_l, swing1LimitDegrees=60, swing2LimitDegrees=60, twistLimitDegrees=20}
 {bone1Name=dyn_outerFront_04_l, bone2Name=dyn_outerFront_03_l, swing1LimitDegrees=60, swing2LimitDegrees=60, twistLimitDegrees=20}
 {bone1Name=dyn_outerFront_05_l, bone2Name=dyn_outerFront_04_l, swing1LimitDegrees=60, swing2LimitDegrees=60, twistLimitDegrees=20}
 {bone1Name=dyn_outerBack_02_l, bone2Name=dyn_outerBackBase_l, swing1LimitDegrees=30, swing2LimitDegrees=30, twistLimitDegrees=10}
 {bone1Name=dyn_outerBack_03_l, bone2Name=dyn_outerBack_02_l, swing1LimitDegrees=60, swing2LimitDegrees=60, twistLimitDegrees=15}
 {bone1Name=dyn_outerBack_04_l, bone2Name=dyn_outerBack_03_l, swing1LimitDegrees=60, swing2LimitDegrees=60, twistLimitDegrees=15}
 {bone1Name=dyn_outerBackMid_02_l, bone2Name=dyn_outerBackMidBase_l, swing1LimitDegrees=30, swing2LimitDegrees=30, twistLimitDegrees=10}
 {bone1Name=dyn_outerBackMid_03_l, bone2Name=dyn_outerBackMid_02_l, swing1LimitDegrees=60, swing2LimitDegrees=60, twistLimitDegrees=15}
 {bone1Name=dyn_outerBackMid_04_l, bone2Name=dyn_outerBackMid_03_l, swing1LimitDegrees=60, swing2LimitDegrees=60, twistLimitDegrees=15}
 {
  bone1Name=dyn_outerFrontMid_04_l
  bone2Name=dyn_outerFront_04_l
  swing1LimitDegrees=45
  swing1Motion=Free
  swing2LimitDegrees=45
  swing2Motion=Free
  twistLimitDegrees=45
  twistMotion=Free
 }
 {
  bone1Name=dyn_outerFrontMid_04_l
  bone2Name=dyn_outerBackMid_04_l
  swing1LimitDegrees=45
  swing1Motion=Free
  swing2LimitDegrees=45
  swing2Motion=Free
  twistLimitDegrees=45
  twistMotion=Free
 }
 {
  bone1Name=dyn_outerBackMidBase_l
  bone2Name=pelvis
  swing1LimitDegrees=45
  swing1Motion=Free
  swing2LimitDegrees=45
  swing2Motion=Free
  twistLimitDegrees=45
  twistMotion=Free
 }
 {
  bone1Name=dyn_outerBackBase_l
  bone2Name=pelvis
  swing1LimitDegrees=45
  swing1Motion=Free
  swing2LimitDegrees=45
  swing2Motion=Free
  twistLimitDegrees=45
  twistMotion=Free
 }
 {bone1Name=thigh_r, bone2Name=pelvis, swing1LimitDegrees=45, swing2LimitDegrees=45, twistLimitDegrees=45}
 {
  bone1Name=calf_r
  bone2Name=thigh_r
  swing1LimitDegrees=45
  swing1Motion=Free
  swing2LimitDegrees=45
  swing2Motion=Free
  twistLimitDegrees=45
  twistMotion=Free
 }
 {bone1Name=dyn_outerFront_05_r, bone2Name=dyn_outerFront_04_r, swing1LimitDegrees=60, swing2LimitDegrees=60, twistLimitDegrees=20}
 {bone1Name=dyn_outerFront_04_r, bone2Name=dyn_outerFront_03_r, swing1LimitDegrees=60, swing2LimitDegrees=60, twistLimitDegrees=20}
 {bone1Name=dyn_outerFront_03_r, bone2Name=dyn_outerFront_02_r, swing1LimitDegrees=60, swing2LimitDegrees=60, twistLimitDegrees=20}
 {bone1Name=dyn_outerFront_02_r, bone2Name=dyn_outerFrontBase_r, swing1LimitDegrees=60, swing2LimitDegrees=60, twistLimitDegrees=15}
 {bone1Name=dyn_outerFrontMid_04_r, bone2Name=dyn_outerFrontMid_03_r, swing1LimitDegrees=60, swing2LimitDegrees=30, twistLimitDegrees=15}
 {bone1Name=dyn_outerFrontMid_03_r, bone2Name=dyn_outerFrontMid_02_r, swing1LimitDegrees=60, swing2LimitDegrees=30, twistLimitDegrees=15}
 {bone1Name=dyn_outerFrontMid_02_r, bone2Name=dyn_outerFrontMidBase_r, swing1LimitDegrees=60, swing2LimitDegrees=60, twistLimitDegrees=15}
 {bone1Name=dyn_outerBackMid_04_r, bone2Name=dyn_outerBackMid_03_r, swing1LimitDegrees=60, swing2LimitDegrees=30, twistLimitDegrees=15}
 {bone1Name=dyn_outerBackMid_03_r, bone2Name=dyn_outerBackMid_02_r, swing1LimitDegrees=60, swing2LimitDegrees=30, twistLimitDegrees=15}
 {bone1Name=dyn_outerBackMid_02_r, bone2Name=dyn_outerBackMidBase_r, swing1LimitDegrees=30, swing2LimitDegrees=30, twistLimitDegrees=10}
 {
  bone1Name=dyn_outerBackMidBase_r
  bone2Name=pelvis
  swing1LimitDegrees=45
  swing1Motion=Free
  swing2LimitDegrees=45
  swing2Motion=Free
  twistLimitDegrees=45
  twistMotion=Free
 }
 {
  bone1Name=dyn_outerBackBase_r
  bone2Name=pelvis
  swing1LimitDegrees=45
  swing1Motion=Free
  swing2LimitDegrees=45
  swing2Motion=Free
  twistLimitDegrees=45
  twistMotion=Free
 }
 {
  bone1Name=dyn_outerFrontMid_04_r
  bone2Name=dyn_outerBackMid_04_r
  swing1LimitDegrees=45
  swing1Motion=Free
  swing2LimitDegrees=45
  swing2Motion=Free
  twistLimitDegrees=45
  twistMotion=Free
 }
 {
  bone1Name=foot_r
  bone2Name=calf_r
  swing1LimitDegrees=45
  swing1Motion=Free
  swing2LimitDegrees=45
  swing2Motion=Free
  twistLimitDegrees=45
  twistMotion=Free
 }
 {
  bone1Name=dyn_outerFrontBase_r
  bone2Name=thigh_twist_01_r
  swing1LimitDegrees=45
  swing1Motion=Free
  swing2LimitDegrees=45
  swing2Motion=Free
  twistLimitDegrees=45
  twistMotion=Free
 }
 {
  bone1Name=dyn_outerFrontMidBase_r
  bone2Name=thigh_twist_01_r
  swing1LimitDegrees=45
  swing1Motion=Free
  swing2LimitDegrees=45
  swing2Motion=Free
  twistLimitDegrees=45
  twistMotion=Free
 }
 {bone1Name=dyn_outerFrontMid_04_r, bone2Name=dyn_outerFront_04_r, swing1LimitDegrees=60, swing2LimitDegrees=60, twistLimitDegrees=60}
 {bone1Name=dyn_outerBack_02_r, bone2Name=dyn_outerBackBase_r, swing1LimitDegrees=45, swing2LimitDegrees=45, twistLimitDegrees=5}
 {bone1Name=dyn_outerBack_03_r, bone2Name=dyn_outerBack_02_r, swing1LimitDegrees=45, swing2LimitDegrees=45, twistLimitDegrees=5}
 {bone1Name=dyn_outerBack_04_r, bone2Name=dyn_outerBack_03_r, swing1LimitDegrees=45, swing2LimitDegrees=45, twistLimitDegrees=5}
 {
  bone1Name=dyn_outerBackMid_02_r
  bone2Name=dyn_outerBack_02_r
  swing1LimitDegrees=45
  swing1Motion=Free
  swing2LimitDegrees=45
  swing2Motion=Free
  twistLimitDegrees=5
 }
 {
  bone1Name=dyn_outerBackMid_03_r
  bone2Name=dyn_outerBack_03_r
  swing1LimitDegrees=45
  swing1Motion=Free
  swing2LimitDegrees=45
  swing2Motion=Free
  twistLimitDegrees=5
 }
]
