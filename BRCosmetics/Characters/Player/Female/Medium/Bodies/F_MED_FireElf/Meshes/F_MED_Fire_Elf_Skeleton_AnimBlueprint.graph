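# Post-process-style AnimBlueprint for the female Fire Elf cosmetic body, adding skirt and arm-sash secondary motion to a pose copied from the parent mesh.
# Constraints and AnimDynamics feed three RigidBody simulations, then crouch and skydive bone corrections, through cached poses.
# The event graph fades cloth blend weight by velocity, vehicle and parachute state, resets cloth around movement changes, and generates random skydive force for the skirts.
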AnimBlueprint /BRCosmetics/Characters/Player/Female/Medium/Bodies/F_MED_FireElf/Meshes/F_MED_Fire_Elf_Skeleton_AnimBlueprint
target_skeleton=&/F_MED_Fire_Elf_Skeleton
graphs [
 AnimGraph {nodes [
   n1 = "Transform (Modify) Bone - Bone: Flap_Front_Driven_Bone"(ComponentPose=Local To Component(LocalPose=Use cached pose 'CrouchCache'()), Alpha=1, Translation="0,0,7", Rotation="0,0,-40", Scale="1,1,1") {
    rotationMode=BMM_Additive
    rotationSpace=BCS_BoneSpace
    translationMode=BMM_Additive
    translationSpace=BCS_BoneSpace
    alphaBoolBlend {blendOption=Linear}
    boneToModify {boneName=Flap_Front_Driven_Bone}
    rotation {roll=-40}
    translation {z=7}
   }
   n2 = Flap Node - Bone: dyn_Flap_Front!(ComponentPose:Component Space Pose Link Structure=@n1, Alpha:float=1) -> (Pose:Component Space Pose Link Structure) {
    is=title FortAnimGraphNode_Flap
    bUseClamp=true
    targetClampMax=100
    alphaBoolBlend {blendOption=Linear}
    sourceDrives [
     {
      bUseRange=false
      drivingCurve=null
      multiplier=0.5
      rangeMax=1
      rangeMin=-1
      remappedMax=1
      remappedMin=0
      sourceComponent=RotationZ
      useQuaternion=true
      sourceBone {boneName=thigh_l}
     }
     {
      bUseRange=false
      drivingCurve=null
      multiplier=0.5
      rangeMax=1
      rangeMin=-1
      remappedMax=1
      remappedMin=0
      sourceComponent=RotationZ
      useQuaternion=true
      sourceBone {boneName=thigh_r}
     }
    ]
    targetBone {boneName=dyn_Flap_Front}
   }
   n3 = "dyn_Flap_Front.rotateZ = thigh_l.rotateZ * -0.5 - Bone Driven Controller"!(ComponentPose:Component Space Pose Link Structure=@n2, Alpha:float=1) -> (Pose:Component Space Pose Link Structure) {
    is=title AnimGraphNode_BoneDrivenController
    bAffectTargetRotationZ=true
    multiplier=-0.5
    sourceComponent=RotationZ
    alphaBoolBlend {blendOption=Linear}
    sourceBone {boneName=thigh_l}
    targetBone {boneName=dyn_Flap_Front}
   }
   "dyn_Flap_Front.rotateZ = thigh_r.rotateZ * 0.5 - Bone Driven Controller"!(ComponentPose:Component Space Pose Link Structure=@n3, Alpha:float=1) -> (Pose:Component Space Pose Link Structure) {
    is=title AnimGraphNode_BoneDrivenController
    bAffectTargetRotationZ=true
    multiplier=0.5
    sourceComponent=RotationZ
    alphaBoolBlend {blendOption=Linear}
    sourceBone {boneName=thigh_r}
    targetBone {boneName=dyn_Flap_Front}
   }
   n4 = "dyn_Flap_FrontOver.rotateZ = curve(thigh_l.rotateZ) - Bone Driven Controller"!(ComponentPose:Component Space Pose Link Structure=Local To Component(LocalPose=Use cached pose 'postRigids'()), Alpha:float=0.75) -> (Pose:Component Space Pose Link Structure) {
    is=title AnimGraphNode_BoneDrivenController
    bAffectTargetRotationZ=true
    drivingCurve=/Game/Characters/Player/Male/XLarge/Bodies/M_XLG_Jim_01/Meshes/SK_M_XLG_Jim_Flap_L_Thigh_Front
    sourceComponent=RotationZ
    alphaBoolBlend {blendOption=Linear}
    sourceBone {boneName=thigh_l}
    targetBone {boneName=dyn_Flap_FrontOver}
   }
   n5 = "dyn_Flap_FrontOver.rotateZ = curve(thigh_r.rotateZ) - Bone Driven Controller"!(ComponentPose:Component Space Pose Link Structure=@n4, Alpha:float=0.75) -> (Pose:Component Space Pose Link Structure) {
    is=title AnimGraphNode_BoneDrivenController
    bAffectTargetRotationZ=true
    drivingCurve=/Game/Characters/Player/Male/XLarge/Bodies/M_XLG_Jim_01/Meshes/SK_M_XLG_Jim_Flap_R_Thigh_Front
    sourceComponent=RotationZ
    alphaBoolBlend {blendOption=Linear}
    sourceBone {boneName=thigh_r}
    targetBone {boneName=dyn_Flap_FrontOver}
   }
   n6 = "dyn_Flap_FrontUnder.rotateZ = curve(thigh_l.rotateZ) - Bone Driven Controller"!(ComponentPose:Component Space Pose Link Structure=@n5, Alpha:float=0.75) -> (Pose:Component Space Pose Link Structure) {
    is=title AnimGraphNode_BoneDrivenController
    bAffectTargetRotationZ=true
    drivingCurve=&SK_M_XLG_Jim_Flap_L_Thigh_Front
    sourceComponent=RotationZ
    alphaBoolBlend {blendOption=Linear}
    sourceBone {boneName=thigh_l}
    targetBone {boneName=dyn_Flap_FrontUnder}
   }
   n7 = "dyn_Flap_FrontUnder.rotateZ = curve(thigh_r.rotateZ) - Bone Driven Controller"!(ComponentPose:Component Space Pose Link Structure=@n6, Alpha:float=0.75) -> (Pose:Component Space Pose Link Structure) {
    is=title AnimGraphNode_BoneDrivenController
    bAffectTargetRotationZ=true
    drivingCurve=&SK_M_XLG_Jim_Flap_R_Thigh_Front
    sourceComponent=RotationZ
    alphaBoolBlend {blendOption=Linear}
    sourceBone {boneName=thigh_r}
    targetBone {boneName=dyn_Flap_FrontUnder}
   }
   n8 = Flap Node - Bone: Flap_Front_Driven_Bone!(ComponentPose:Component Space Pose Link Structure=@n7, Alpha:float=0.95) -> (Pose:Component Space Pose Link Structure) {
    is=title FortAnimGraphNode_Flap
    bUseClamp=true
    targetClampMax=3.14
    alphaBoolBlend {blendOption=Linear}
    sourceDrives [
     {
      bUseRange=false
      drivingCurve=null
      multiplier=1
      rangeMax=1
      rangeMin=-1
      remappedMax=1
      remappedMin=0
      sourceComponent=RotationZ
      useQuaternion=true
      sourceBone {boneName=thigh_l}
     }
     {
      bUseRange=false
      drivingCurve=null
      multiplier=1
      rangeMax=1
      rangeMin=-1
      remappedMax=1
      remappedMin=0
      sourceComponent=RotationZ
      useQuaternion=true
      sourceBone {boneName=thigh_r}
     }
    ]
    targetBone {boneName=Flap_Front_Driven_Bone}
   }
   n9 = "Transform (Modify) Bone - Bone: dyn_Flap_FrontOver"(ComponentPose=@n8, Alpha=1, Translation="0,0,0", Rotation="0,0,0", Scale="1,1,1") {rotationMode=BMM_Additive, rotationSpace=BCS_BoneSpace, alphaBoolBlend {blendOption=Linear}, boneToModify {boneName=dyn_Flap_FrontOver}}
   Component To Local(ComponentPose=@n9)
   n10 = Look At - Bone: dyn_front_topFlap_l(Alpha=1) {
    bUseLookUpAxis=true
    lookAtClamp=30
    alphaBoolBlend {blendOption=Linear}
    boneToModify {boneName=dyn_front_topFlap_l}
    lookAtLocation=<0 -10>
    lookAtTarget {boneReference {boneName=dyn_cloth_fr_02}}
    lookAt_Axis {axis=<1 0>}
    lookUp_Axis {axis {x=0, z=1}}
   }
   Look At - Bone: dyn_front_topFlap_r(ComponentPose=@n10, Alpha=1) {
    bUseLookUpAxis=true
    lookAtClamp=30
    alphaBoolBlend {blendOption=Linear}
    boneToModify {boneName=dyn_front_topFlap_r}
    lookAtLocation=<0 -3>
    lookAtTarget {boneReference {boneName=dyn_cloth_fl_02}}
    lookAt_Axis {axis=<1 0>}
    lookUp_Axis {axis {x=0, z=1}}
   }
   n11 = "Transform (Modify) Bone - Bone: dyn_outerFrontBase_r"!(
    ComponentPose=Local To Component(LocalPose=Use cached pose 'postSkydive'())
    bAlphaBoolEnabled:bool=false
    Translation="0,0,0"
    Rotation="0,-15,0"
    Scale="1,1,1"
   ) -> (Pose) {
    alphaInputType=Bool
    rotationMode=BMM_Additive
    rotationSpace=BCS_BoneSpace
    translationMode=BMM_Additive
    translationSpace=BCS_BoneSpace
    alphaBoolBlend {blendOption=Linear}
    boneToModify {boneName=dyn_outerFrontBase_r}
    rotation {yaw=-15}
   }
   n12 = "Transform (Modify) Bone - Bone: dyn_outerFrontMidBase_r"!(ComponentPose=@n11, bAlphaBoolEnabled:bool=false, Translation="0,0,0", Rotation="0,-15,0", Scale="1,1,1") -> (Pose) {
    alphaInputType=Bool
    rotationMode=BMM_Additive
    rotationSpace=BCS_BoneSpace
    translationMode=BMM_Additive
    translationSpace=BCS_BoneSpace
    alphaBoolBlend {blendOption=Linear}
    boneToModify {boneName=dyn_outerFrontMidBase_r}
    rotation {yaw=-15}
   }
   n13 = "Transform (Modify) Bone - Bone: dyn_outerBackBase_r"!(ComponentPose=@n12, bAlphaBoolEnabled:bool=false, Translation="0,0,0", Rotation="0,-20,0", Scale="1,1,1") -> (Pose) {
    alphaInputType=Bool
    rotationMode=BMM_Additive
    rotationSpace=BCS_BoneSpace
    translationMode=BMM_Additive
    translationSpace=BCS_BoneSpace
    alphaBoolBlend {blendOption=Linear}
    boneToModify {boneName=dyn_outerBackBase_r}
    rotation {yaw=-20}
   }
   n14 = "Transform (Modify) Bone - Bone: dyn_outerBackMidBase_r"!(ComponentPose=@n13, bAlphaBoolEnabled:bool=false, Translation="0,0,0", Rotation="0,-20,0", Scale="1,1,1") -> (Pose) {
    alphaInputType=Bool
    rotationMode=BMM_Additive
    rotationSpace=BCS_BoneSpace
    translationMode=BMM_Additive
    translationSpace=BCS_BoneSpace
    alphaBoolBlend {blendOption=Linear}
    boneToModify {boneName=dyn_outerBackMidBase_r}
    rotation {yaw=-20}
   }
   n15 = "Transform (Modify) Bone - Bone: dyn_outerFrontBase_l"!(ComponentPose=@n14, bAlphaBoolEnabled:bool=false, Translation="0,0,0", Rotation="0,-15,0", Scale="1,1,1") -> (Pose) {
    alphaInputType=Bool
    rotationMode=BMM_Additive
    rotationSpace=BCS_BoneSpace
    translationMode=BMM_Additive
    translationSpace=BCS_BoneSpace
    alphaBoolBlend {blendOption=Linear}
    boneToModify {boneName=dyn_outerFrontBase_l}
    rotation {yaw=-15}
   }
   n16 = "Transform (Modify) Bone - Bone: dyn_outerFrontMidBase_l"!(ComponentPose=@n15, bAlphaBoolEnabled:bool=false, Translation="0,0,0", Rotation="0,-15,0", Scale="1,1,1") -> (Pose) {
    alphaInputType=Bool
    rotationMode=BMM_Additive
    rotationSpace=BCS_BoneSpace
    translationMode=BMM_Additive
    translationSpace=BCS_BoneSpace
    alphaBoolBlend {blendOption=Linear}
    boneToModify {boneName=dyn_outerFrontMidBase_l}
    rotation {yaw=-15}
   }
   n17 = "Transform (Modify) Bone - Bone: dyn_outerBackBase_l"!(ComponentPose=@n16, bAlphaBoolEnabled:bool=false, Translation="0,0,0", Rotation="0,-20,0", Scale="1,1,1") -> (Pose) {
    alphaInputType=Bool
    rotationMode=BMM_Additive
    rotationSpace=BCS_BoneSpace
    translationMode=BMM_Additive
    translationSpace=BCS_BoneSpace
    alphaBoolBlend {blendOption=Linear}
    boneToModify {boneName=dyn_outerBackBase_l}
    rotation {yaw=-20}
   }
   n18 = "Transform (Modify) Bone - Bone: dyn_outerBackMidBase_l"!(ComponentPose=@n17, bAlphaBoolEnabled:bool=false, Translation="0,0,0", Rotation="0,-20,0", Scale="1,1,1") -> (Pose) {
    alphaInputType=Bool
    rotationMode=BMM_Additive
    rotationSpace=BCS_BoneSpace
    translationMode=BMM_Additive
    translationSpace=BCS_BoneSpace
    alphaBoolBlend {blendOption=Linear}
    boneToModify {boneName=dyn_outerBackMidBase_l}
    rotation {yaw=-20}
   }
   Output Pose(Result=Component To Local(ComponentPose=@n18)) {layerGroup=None}
   n19 = "Constraint - dyn_outerFrontBaseConst_l (PPPP)"!(
    ComponentPose=Local To Component(LocalPose=Copy Pose From Mesh() {bUseAttachedParent=true})
    Alpha=0
    ConstraintWeights_0=0.0007
    ConstraintWeights_1:float=0.0961
    ConstraintWeights_2:float=0.0024
    ConstraintWeights_3:float=0.9008
   ) -> (Pose) {
    alphaBoolBlend {blendOption=Linear}
    boneToModify {boneName=dyn_outerFrontBaseConst_l}
    constraintSetup [{targetBone {boneName=thigh_r}}, {targetBone {boneName=pelvis}}, {targetBone {boneName=thigh_twist_01_r}}, {targetBone {boneName=thigh_l}}]
    constraintWeights [1,1,1,1]
   }
   n20 = "Constraint - dyn_outerFrontMidConst_l (PPP)"!(ComponentPose=@n19, Alpha=0, ConstraintWeights_0=0.0811, ConstraintWeights_1:float=0.0033, ConstraintWeights_2:float=0.9155) -> (Pose) {
    alphaBoolBlend {blendOption=Linear}
    boneToModify {boneName=dyn_outerFrontMidConst_l}
    constraintSetup [{targetBone {boneName=pelvis}}, {targetBone {boneName=thigh_twist_01_l}}, {targetBone {boneName=thigh_l}}]
    constraintWeights [1,1,1]
   }
   n21 = "Constraint - dyn_outerBackMidBase_l (PPPP)"!(
    ComponentPose=@n20
    Alpha=1
    ConstraintWeights_0=0.0004
    ConstraintWeights_1:float=0.1205
    ConstraintWeights_2:float=0.0066
    ConstraintWeights_3:float=0.8726
   ) -> (Pose) {
    alphaBoolBlend {blendOption=Linear}
    boneToModify {boneName=dyn_outerBackMidBase_l}
    constraintSetup [{targetBone {boneName=thigh_r}}, {targetBone {boneName=pelvis}}, {targetBone {boneName=thigh_twist_01_r}}, {targetBone {boneName=thigh_l}}]
    constraintWeights [1,1,1,1]
   }
   n22 = "Constraint - dyn_outerBackBase_l (PPPP)"!(
    ComponentPose=@n21
    Alpha=1
    ConstraintWeights_0=0.0405
    ConstraintWeights_1:float=0.4868
    ConstraintWeights_2:float=0.0027
    ConstraintWeights_3:float=0.47
   ) -> (Pose) {
    alphaBoolBlend {blendOption=Linear}
    boneToModify {boneName=dyn_outerBackBase_l}
    constraintSetup [{targetBone {boneName=thigh_r}}, {targetBone {boneName=pelvis}}, {targetBone {boneName=thigh_twist_01_r}}, {targetBone {boneName=thigh_l}}]
    constraintWeights [1,1,1,1]
   }
   n23 = "Constraint - dyn_outerFrontConst_r (PPPP)"!(
    ComponentPose=@n22
    Alpha=1
    ConstraintWeights_0=0.9019
    ConstraintWeights_1:float=0.0935
    ConstraintWeights_2:float=0.0025
    ConstraintWeights_3:float=0.0017
   ) -> (Pose) {
    alphaBoolBlend {blendOption=Linear}
    boneToModify {boneName=dyn_outerFrontConst_r}
    constraintSetup [{targetBone {boneName=thigh_r}}, {targetBone {boneName=pelvis}}, {targetBone {boneName=thigh_twist_01_r}}, {targetBone {boneName=thigh_l}}]
    constraintWeights [1,1,1,1]
   }
   n24 = "Constraint - dyn_outerFrontMidConst_r (PPP)"!(ComponentPose=@n23, Alpha=1, ConstraintWeights_0=0.0811, ConstraintWeights_1:float=0.0033, ConstraintWeights_2:float=0.9155) -> (Pose) {
    alphaBoolBlend {blendOption=Linear}
    boneToModify {boneName=dyn_outerFrontMidConst_r}
    constraintSetup [{targetBone {boneName=pelvis}}, {targetBone {boneName=thigh_twist_01_r}}, {targetBone {boneName=thigh_r}}]
    constraintWeights [1,1,1]
   }
   n25 = "Constraint - dyn_outerBackMidBase_r (PPP)"!(ComponentPose=@n24, Alpha=1, ConstraintWeights_0=0.8706, ConstraintWeights_1:float=0.1205, ConstraintWeights_2:float=0.0066) -> (Pose) {
    alphaBoolBlend {blendOption=Linear}
    boneToModify {boneName=dyn_outerBackMidBase_r}
    constraintSetup [{targetBone {boneName=thigh_r}}, {targetBone {boneName=pelvis}}, {targetBone {boneName=thigh_l}}]
    constraintWeights [1,1,1]
   }
   n26 = "Constraint - dyn_outerBackBase_r (PPPPP)"!(
    ComponentPose=@n25
    Alpha=1
    ConstraintWeights_0=0.6372
    ConstraintWeights_1:float=0.3207
    ConstraintWeights_2:float=0
    ConstraintWeights_3:float=0.0342
    ConstraintWeights_4:float=0.008
   ) -> (Pose) {
    alphaBoolBlend {blendOption=Linear}
    boneToModify {boneName=dyn_outerBackBase_r}
    constraintSetup [
     {targetBone {boneName=thigh_r}}
     {targetBone {boneName=pelvis}}
     {targetBone {boneName=thigh_twist_01_r}}
     {targetBone {boneName=thigh_l}}
     {targetBone {boneName=thigh_twist_01_l}}
    ]
    constraintWeights [1,1,1,1,1]
   }
   n27 = "Constraint - thigh_half_r (R)"(ComponentPose=@n26, Alpha=0.5, ConstraintWeights_0=1) {
    alphaBoolBlend {blendOption=Linear}
    boneToModify {boneName=thigh_half_r}
    constraintSetup [{transformType=Rotation, targetBone {boneName=thigh_r}}]
    constraintWeights [1]
   }
   n28 = "Constraint - thigh_half_l (R)"(ComponentPose=@n27, Alpha=0.5, ConstraintWeights_0=1) {
    alphaBoolBlend {blendOption=Linear}
    boneToModify {boneName=thigh_half_l}
    constraintSetup [{transformType=Rotation, targetBone {boneName=thigh_l}}]
    constraintWeights [1]
   }
   n29 = "Constraint - dyn_armSashBase (PPP)"!(ComponentPose=@n28, Alpha=1, ConstraintWeights_0=0.0639, ConstraintWeights_1:float=0.4729, ConstraintWeights_2:float=0.4631) -> (Pose) {
    alphaBoolBlend {blendOption=Linear}
    boneToModify {boneName=dyn_armSashBase}
    constraintSetup [{targetBone {boneName=upperarm_r}}, {targetBone {boneName=upperarm_twist_01_r}}, {targetBone {boneName=upperarm_twist_02_r}}]
    constraintWeights [1,1,1]
   }
   n30 = "Transform (Modify) Bone - Bone: calf_twist_02_l"(ComponentPose=@n29, Alpha=1, Translation="0,0,0", Rotation="0,0,0", Scale="0.9,0.9,0.9") {
    rotationSpace=BCS_BoneSpace
    scaleMode=BMM_Replace
    translationSpace=BCS_BoneSpace
    alphaBoolBlend {blendOption=Linear}
    boneToModify {boneName=calf_twist_02_l}
    scale=<0.9 0.9 0.9>
   }
   n31 = "Transform (Modify) Bone - Bone: dyn_pelvisShift"(ComponentPose=@n30, Alpha=1, Translation="0,0,0", Rotation="0,0,0", Scale="0.92,0.92,0.92") {
    rotationSpace=BCS_BoneSpace
    scaleMode=BMM_Replace
    translationSpace=BCS_BoneSpace
    alphaBoolBlend {blendOption=Linear}
    boneToModify {boneName=dyn_pelvisShift}
    scale=<0.92 0.92 0.92>
   }
   n32 = Anim Dynamics - Bone: dyn_armSashHi(ComponentPose=@n31, Alpha=0.75, ExternalForce="0,0,0") {
    alpha=0.9
    angularSpringConstant=1500
    bAngularSpring=true
    bGravityOverrideInSimSpace=true
    numSolverIterationsPostUpdate=4
    numSolverIterationsPreUpdate=16
    simulationSpace=Actor
    alphaBoolBlend {blendOption=Linear}
    boundBone {boneName=dyn_armSashHi}
    physicsBodyDefinitions [{
      sphereCollisionRadius=0
      boundBone {boneName=dyn_armSashHi}
      boxExtents=<5 15 5>
      constraintSetup {angularTargetAxis=AxisZ, angularLimitsMax=<0 15 15>, angularLimitsMin=<0 -15 -15>, angularTarget=<0 0 1>}
      localJointOffset=<0 0 0>
     }]
   }
   n33 = Anim Dynamics - Bone: dyn_armSashLow(ComponentPose=@n32, Alpha=0.75, ExternalForce="0,0,0") {
    alpha=0.9
    angularSpringConstant=1500
    bAngularSpring=true
    bGravityOverrideInSimSpace=true
    numSolverIterationsPostUpdate=4
    numSolverIterationsPreUpdate=16
    simulationSpace=Actor
    alphaBoolBlend {blendOption=Linear}
    boundBone {boneName=dyn_armSashLow}
    physicsBodyDefinitions [{
      sphereCollisionRadius=0
      boundBone {boneName=dyn_armSashLow}
      boxExtents=<5 5 15>
      constraintSetup {angularTargetAxis=AxisY, angularLimitsMax=<0 15 15>, angularLimitsMin=<0 -15 -15>, angularTarget=<0 1 0>}
      localJointOffset=<0 0 0>
     }]
   }
   Save cached pose 'PreRigids'(Pose=Component To Local(ComponentPose=@n33)) {cacheName=PreRigids}
   n34 = Get skydiving vector force() -> (skydiving vector force:Vector)
   n35 = RigidBody(ComponentPose=Local To Component(LocalPose=Use cached pose 'PreRigids'()), ExternalForce=@n34) {
    bClampLinearTranslationLimitToRefPose=true
    bOverrideWorldGravity=true
    overridePhysicsAsset=&/F_MED_Fire_Elf_LSkirt_Physics
    simulationSpace=BaseBoneSpace
    alphaBoolBlend {blendOption=Linear}
    baseBoneRef {boneName=thigh_half_l}
    componentLinearAccScale=<0.05 0.05 0.05>
    componentLinearVelScale=<0.8 0.8 0.8>
    overrideWorldGravity {z=-980}
   }
   n36 = RigidBody(ComponentPose=@n35, ExternalForce=@n34) {
    bClampLinearTranslationLimitToRefPose=true
    bOverrideWorldGravity=true
    overridePhysicsAsset=&/F_MED_Fire_Elf_RSkirtB_Physics
    simulationSpace=BaseBoneSpace
    alphaBoolBlend {blendOption=Linear}
    baseBoneRef {boneName=thigh_half_r}
    componentLinearAccScale=<0.15 0.15 0.15>
    componentLinearVelScale=<0.8 0.8 0.8>
    overrideWorldGravity {z=-980}
   }
   n37 = RigidBody(ComponentPose=@n36) {
    bClampLinearTranslationLimitToRefPose=true
    bOverrideWorldGravity=true
    overridePhysicsAsset=&/F_MED_Fire_Elf_InnerCollision
    simulationSpace=BaseBoneSpace
    alphaBoolBlend {blendOption=Linear}
    baseBoneRef {boneName=pelvis}
    componentAppliedLinearAccClamp=<100 100 100>
    componentLinearAccScale=<0.15 0.15 0.15>
    componentLinearVelScale=<0.8 0.8 0.8>
    overrideWorldGravity {z=-400}
   }
   n38 = Get bIsSkyDiving()
   n39 = "Transform (Modify) Bone - Bone: dyn_outerBack_04_l"!(ComponentPose=@n37, bAlphaBoolEnabled:bool=@n38, Translation="0,0,0", Rotation="0,-5,0", Scale="1,1,1") -> (Pose) {
    alphaInputType=Bool
    rotationMode=BMM_Additive
    rotationSpace=BCS_BoneSpace
    alphaBoolBlend {blendInTime=0.5, blendOption=Linear, blendOutTime=0.5}
    alphaScaleBiasClamp {bInterpResult=true, interpSpeedDecreasing=5, interpSpeedIncreasing=5}
    boneToModify {boneName=dyn_outerBack_04_l}
    rotation {yaw=-5}
   }
   n40 = "Transform (Modify) Bone - Bone: dyn_outerBackBase_r"(ComponentPose=@n39, bAlphaBoolEnabled=@n38, Translation="0,0,0", Rotation="0,-5,0", Scale="1,1,1") {
    alphaInputType=Bool
    rotationMode=BMM_Additive
    rotationSpace=BCS_BoneSpace
    alphaBoolBlend {blendInTime=0.5, blendOption=Linear, blendOutTime=0.5}
    alphaScaleBiasClamp {bInterpResult=true, interpSpeedDecreasing=5, interpSpeedIncreasing=5}
    boneToModify {boneName=dyn_outerBackBase_r}
    rotation {yaw=-5}
   }
   n41 = "Transform (Modify) Bone - Bone: dyn_outerFrontMidBase_r"(ComponentPose=@n40, bAlphaBoolEnabled=@n38, Translation="0,0,0", Rotation="0,-5,0", Scale="1,1,1") {
    alphaInputType=Bool
    rotationMode=BMM_Additive
    rotationSpace=BCS_BoneSpace
    alphaBoolBlend {blendInTime=0.5, blendOption=Linear, blendOutTime=0.5}
    alphaScaleBiasClamp {bInterpResult=true, interpSpeedDecreasing=5, interpSpeedIncreasing=5}
    boneToModify {boneName=dyn_outerFrontMidBase_r}
    rotation {yaw=-5}
   }
   n42 = "Transform (Modify) Bone - Bone: dyn_outerBackMidBase_r"(ComponentPose=@n41, bAlphaBoolEnabled=@n38, Translation="0,0,0", Rotation="0,-5,0", Scale="1,1,1") {
    alphaInputType=Bool
    rotationMode=BMM_Additive
    rotationSpace=BCS_BoneSpace
    alphaBoolBlend {blendInTime=0.5, blendOption=Linear, blendOutTime=0.5}
    alphaScaleBiasClamp {bInterpResult=true, interpSpeedDecreasing=5, interpSpeedIncreasing=5}
    boneToModify {boneName=dyn_outerBackMidBase_r}
    rotation {yaw=-5}
   }
   n43 = "Transform (Modify) Bone - Bone: dyn_outerBackBase_l"(ComponentPose=@n42, bAlphaBoolEnabled=@n38, Translation="0,0,0", Rotation="0,-20,0", Scale="1,1,1") {
    alphaInputType=Bool
    rotationMode=BMM_Additive
    rotationSpace=BCS_BoneSpace
    alphaBoolBlend {blendInTime=0.5, blendOption=Linear, blendOutTime=0.5}
    alphaScaleBiasClamp {bInterpResult=true, interpSpeedDecreasing=5, interpSpeedIncreasing=5}
    boneToModify {boneName=dyn_outerBackBase_l}
    rotation {yaw=-20}
   }
   n44 = "Transform (Modify) Bone - Bone: dyn_outerBackMidBase_l"(ComponentPose=@n43, bAlphaBoolEnabled=@n38, Translation="0,0,0", Rotation="0,-20,0", Scale="1,1,1") {
    alphaInputType=Bool
    rotationMode=BMM_Additive
    rotationSpace=BCS_BoneSpace
    alphaBoolBlend {blendInTime=0.5, blendOption=Linear, blendOutTime=0.5}
    alphaScaleBiasClamp {bInterpResult=true, interpSpeedDecreasing=5, interpSpeedIncreasing=5}
    boneToModify {boneName=dyn_outerBackMidBase_l}
    rotation {yaw=-20}
   }
   Save cached pose 'postRigids'(Pose=Component To Local(ComponentPose=@n44)) {cacheName=postRigids}
   n45 = "Transform (Modify) Bone - Bone: dyn_outerFrontBase_l"!(
    ComponentPose=Local To Component(LocalPose=Use cached pose 'postRigids'())
    Alpha:float=1
    Translation="-0.000,0.000,2"
    Rotation="0,0,0"
    Scale="1,1,1"
   ) -> (Pose) {
    rotationMode=BMM_Additive
    rotationSpace=BCS_BoneSpace
    translationMode=BMM_Additive
    translationSpace=BCS_BoneSpace
    alphaBoolBlend {blendOption=Linear}
    boneToModify {boneName=dyn_outerFrontBase_l}
    translation {z=2}
   }
   n46 = "Transform (Modify) Bone - Bone: dyn_outerBackMid_03_r"(ComponentPose=@n45, Alpha=1, Translation="0,-5,5", Rotation="0,0,0", Scale="1,1,1") {
    rotationMode=BMM_Additive
    rotationSpace=BCS_BoneSpace
    translationMode=BMM_Additive
    translationSpace=BCS_BoneSpace
    alphaBoolBlend {blendOption=Linear}
    boneToModify {boneName=dyn_outerBackMid_03_r}
    translation {y=-5, z=5}
   }
   n47 = Component To Local(ComponentPose=@n46)
   n48 = Use cached pose 'postRigids'()
   n49 = Get bIsCrouching()
   n50 = "Transform (Modify) Bone - Bone: dyn_outerFrontBase_r"!(
    ComponentPose=Local To Component(LocalPose=Blend Poses by bool(BlendPose_0=@n47, BlendPose_1=@n48, BlendTime_0=0, BlendTime_1=1, bActiveValue=@n49) {blendType=Linear})
    Alpha:float=1
    Translation="-0.000,-1,0"
    Rotation="0,2,0"
    Scale="1,1,1"
   ) -> (Pose) {
    rotationMode=BMM_Additive
    rotationSpace=BCS_BoneSpace
    translationMode=BMM_Additive
    translationSpace=BCS_BoneSpace
    alphaBoolBlend {blendOption=Linear}
    boneToModify {boneName=dyn_outerFrontBase_r}
    rotation {yaw=2}
    translation {y=-1}
   }
   Save cached pose 'postCrouch'(Pose=Component To Local(ComponentPose=@n50)) {cacheName=postCrouch}
   n51 = Get SkydiveAlpha() -> (SkydiveAlpha:double)
   n52 = "Transform (Modify) Bone - Bone: dyn_outerFrontMidBase_l"!(
    ComponentPose=Local To Component(LocalPose=Use cached pose 'postCrouch'())
    Alpha:float=@n51
    Translation="0,0,0"
    Rotation="6,-30,0"
    Scale="1,1,1"
   ) -> (Pose) {
    rotationMode=BMM_Additive
    rotationSpace=BCS_BoneSpace
    translationSpace=BCS_BoneSpace
    alphaBoolBlend {blendOption=Linear}
    alphaScaleBiasClamp {bInterpResult=true}
    boneToModify {boneName=dyn_outerFrontMidBase_l}
    rotation {pitch=6, yaw=-30}
   }
   n53 = "Transform (Modify) Bone - Bone: dyn_outerFrontBase_l"(ComponentPose=@n52, Alpha=@n51, Translation="0,0,0", Rotation="6,-30,0", Scale="1,1,1") {
    rotationMode=BMM_Additive
    rotationSpace=BCS_BoneSpace
    translationSpace=BCS_BoneSpace
    alphaBoolBlend {blendOption=Linear}
    alphaScaleBiasClamp {bInterpResult=true}
    boneToModify {boneName=dyn_outerFrontBase_l}
    rotation {pitch=6, yaw=-30}
   }
   n54 = "Transform (Modify) Bone - Bone: dyn_outerBackBase_l"!(ComponentPose=@n53, Alpha:float=@n51, Translation="0,1,0", Rotation="6,-30,0", Scale="1,1,1") -> (Pose) {
    rotationMode=BMM_Additive
    rotationSpace=BCS_BoneSpace
    translationMode=BMM_Additive
    translationSpace=BCS_BoneSpace
    alphaBoolBlend {blendOption=Linear}
    alphaScaleBiasClamp {bInterpResult=true}
    boneToModify {boneName=dyn_outerBackBase_l}
    rotation {pitch=6, yaw=-30}
    translation {y=1}
   }
   n55 = "Transform (Modify) Bone - Bone: dyn_outerBackMidBase_l"!(ComponentPose=@n54, Alpha:float=@n51, Translation="0,0,0", Rotation="6,-30,0", Scale="1,1,1") -> (Pose) {
    rotationMode=BMM_Additive
    rotationSpace=BCS_BoneSpace
    translationSpace=BCS_BoneSpace
    alphaBoolBlend {blendOption=Linear}
    alphaScaleBiasClamp {bInterpResult=true}
    boneToModify {boneName=dyn_outerBackMidBase_l}
    rotation {pitch=6, yaw=-30}
   }
   n56 = "Transform (Modify) Bone - Bone: dyn_outerFrontMidBase_r"!(ComponentPose=@n55, Alpha:float=@n51, Translation="0,0,0", Rotation="6,-15,0", Scale="1,1,1") -> (Pose) {
    rotationMode=BMM_Additive
    rotationSpace=BCS_BoneSpace
    translationSpace=BCS_BoneSpace
    alphaBoolBlend {blendOption=Linear}
    alphaScaleBiasClamp {bInterpResult=true}
    boneToModify {boneName=dyn_outerFrontMidBase_r}
    rotation {pitch=6, yaw=-15}
   }
   n57 = "Transform (Modify) Bone - Bone: dyn_outerFrontBase_r"(ComponentPose=@n56, Alpha=@n51, Translation="0,0,0", Rotation="6,-15,0", Scale="1,1,1") {
    rotationMode=BMM_Additive
    rotationSpace=BCS_BoneSpace
    translationSpace=BCS_BoneSpace
    alphaBoolBlend {blendOption=Linear}
    alphaScaleBiasClamp {bInterpResult=true}
    boneToModify {boneName=dyn_outerFrontBase_r}
    rotation {pitch=6, yaw=-15}
   }
   n58 = "Transform (Modify) Bone - Bone: dyn_outerBackBase_r"!(ComponentPose=@n57, Alpha:float=@n51, Translation="0,-1,0", Rotation="6,-15,0", Scale="1,1,1") -> (Pose) {
    rotationMode=BMM_Additive
    rotationSpace=BCS_BoneSpace
    translationMode=BMM_Additive
    translationSpace=BCS_BoneSpace
    alphaBoolBlend {blendOption=Linear}
    alphaScaleBiasClamp {bInterpResult=true}
    boneToModify {boneName=dyn_outerBackBase_r}
    rotation {pitch=6, yaw=-15}
    translation {y=-1}
   }
   n59 = "Transform (Modify) Bone - Bone: dyn_outerBackMidBase_r"!(ComponentPose=@n58, Alpha:float=@n51, Translation="0,0,0", Rotation="6,-15,0", Scale="1,1,1") -> (Pose) {
    rotationMode=BMM_Additive
    rotationSpace=BCS_BoneSpace
    translationSpace=BCS_BoneSpace
    alphaBoolBlend {blendOption=Linear}
    alphaScaleBiasClamp {bInterpResult=true}
    boneToModify {boneName=dyn_outerBackMidBase_r}
    rotation {pitch=6, yaw=-15}
   }
   n60 = "Transform (Modify) Bone - Bone: dyn_outerBack_03_r"!(ComponentPose=@n59, bAlphaBoolEnabled:bool=Get bIsSkyDiving(), Translation="0,-5,0", Rotation="6,-10,0", Scale="1,1,1") -> (Pose) {
    alphaInputType=Bool
    rotationMode=BMM_Additive
    rotationSpace=BCS_BoneSpace
    translationMode=BMM_Additive
    translationSpace=BCS_BoneSpace
    alphaBoolBlend {blendOption=Linear}
    alphaScaleBiasClamp {bInterpResult=true}
    boneToModify {boneName=dyn_outerBack_03_r}
    rotation {pitch=6, yaw=-10}
    translation {y=-5}
   }
   Save cached pose 'postSkydive'(Pose=Component To Local(ComponentPose=@n60)) {cacheName=postSkydive}
   vector * float(
    A=Break Fort Anim Input Skydiving External Force!(FortAnimInput_SkydivingExternalForce=Get SkydivingExternalForceInput()) -> (
     SkydivingVectorForce=~
     HeadToPelvisDirection=~
     FloatingMultiplier=~
     FloatingAdditive=~
     DivingMultiplier=~
     DivingAdditive=~
     ParachutingMultiplier=~
     ParachutingAdditive=~
    )
    B=20
   )
  ]}
 EventGraph {nodes [
   ResetDynamics(InTeleportType=TeleportPhysics)
   > ResetDynamics(InTeleportType=ResetPhysics)
   SetClothMaxDistanceScale!(execute, self:Skeletal Mesh Component Object Reference=GetPartSkeletalMeshComponent(), Scale:float=0.05) -> (then)
   > SetAnimDriveSpringStiffness!(execute, self:Clothing Simulation Interactor Object Reference, InStiffness:float=1) -> (then)
   > SetAnimDriveDamperStiffness!(execute, self:Clothing Simulation Interactor Nv Object Reference, InStiffness:float=100) -> (then)
   Delay(Duration=0.5)
   > Branch(Condition="float < float"!(A:double=Get PawnVelocityZ(), B:double=0) -> (ReturnValue:bool) {is=title K2Node_CallFunction})
   > Delay(Duration=0.5)
   > ForceClothNextUpdateTeleportAndReset(self:Skeletal Mesh Component Object Reference=Get MeshToCopy_0() -> (MeshToCopy_0:Skeletal Mesh Component Object Reference))
   n1 = Event BlueprintUpdateAnimation()
   > n2 = Sequence!(execute) -> (then_0, then_1, then_2:Exec, then_3:Exec, then_4:Exec)
   Branch(execute=@n2.then_1, Condition=Get bClothEnabled())
   n3 = "float > float"(A=Get PawnVelocityZ(), B=200) {is=title K2Node_CallFunction}
   n4 = Get bIsSkyDiving()
   > n5 = Branch(Condition=NOT Boolean(A=AND Boolean(A=@n3, B=@n4)))
   > Set ClothBlendDelta(ClothBlendDelta:double=FInterpTo(Current=Get ClothBlendDelta() -> (ClothBlendDelta:double), Target=1, DeltaTime=@n1.DeltaTimeX, InterpSpeed=0.85)) -> (Output_Get:double)
   Set ClothBlendDelta(execute=@n5.else, ClothBlendDelta=FInterpTo(Current=Get ClothBlendDelta(), Target=0, DeltaTime=@n1.DeltaTimeX, InterpSpeed=10))
   Branch(execute=@n2.then_2, Condition=Get bClothEnabled())
   n6 = Get bIsSkyDiving()
   n7 = Get PawnVelocityZ()
   > n8 = Branch(Condition=AND Boolean(A="float < float"(A=@n7, B=-700), B=NOT Boolean(A=@n6)))
   > Set ClothBlendDeltaFall(ClothBlendDeltaFall:double=FInterpTo(Current=Get ClothBlendDeltaFall() -> (ClothBlendDeltaFall:double), Target=0, DeltaTime=@n1.DeltaTimeX, InterpSpeed=3)) -> (Output_Get:double)
   Set ClothBlendDeltaFall(execute=@n8.else, ClothBlendDeltaFall=FInterpTo(Current=Get ClothBlendDeltaFall(), Target=1, DeltaTime=@n1.DeltaTimeX, InterpSpeed=10))
   n9 = Branch(execute=@n2.then_3, Condition=Get bCanPlayCustomAnimations() -> (bCanPlayCustomAnimations:bool))
   n10 = Branch(execute=[@n9.else, @n9.then], Condition=Get bIsSkyDiving())
   > n11 = Branch(Condition=Get bIsSkydiveFloating())
   n12 = RandomFloatInRangeFromStream!(Stream:"Random Stream Structure (by ref)"=Get random_stream() -> (random_stream:Random Stream Structure), Min:float=-20000, Max:float=20000) -> (ReturnValue:float)
   n13 = MakeVector(X=@n12, Y=@n12)
   > Set skydiving vector force(skydiving vector force:Vector=@n13) -> (Output_Get:Vector)
   n14 = Branch(execute=@n11.else, Condition=Get bIsSkydiveDiveMode())
   n15 = Get Direction Between Sockets!(
    Component:Skeletal Mesh Component Object Reference=GetOwnerSkeletalMeshComponent()
    SocketOrBoneNameFrom:Name=head
    SocketOrBoneNameTo:Name=pelvis
   ) -> (ReturnValue:Vector) {is=title K2Node_CallFunction}
   > Set skydiving vector force(skydiving vector force=vector * vector(A:Vector=vector * vector(A=@n13, B=@n15), B:Vector="0.2,0.2,0.2") -> (ReturnValue:Vector) {is=title K2Node_CommutativeAssociativeBinaryOperator})
   Set skydiving vector force(execute=@n14.else, skydiving vector force=@n13)
   Set skydiving vector force(execute=@n10.else, skydiving vector force=vector * vector(A="0, 0, 0", B="0, 0, 0"))
   Set Rigid Body Not Enabled(execute=@n2.then_4, Rigid Body Not Enabled:bool=NOT Boolean(A=Get bRigidBodyEnabled())) -> (Output_Get:bool)
   Event OnCharacterCustomizationLoaded()
   n16 = CheckForBackbling()
   > n17 = Set IsWearingBackbling(IsWearingBackbling:bool=@n16) -> (Output_Get:bool)
   > n18 = Set IsWearingCape(IsWearingCape:bool=@n16.bIsBackblingLong) -> (Output_Get:bool)
   > Branch(Condition=@n18)
   > Is Valid(InputObject=TryGetPawnOwner())
   > n19 = Sequence!(execute) -> (then_0, then_1)
   n20 = Get bIsSkyDiving()
   > Set SkydiveAlpha(SkydiveAlpha:double="To Float (Boolean)"(InBool=@n20)) -> (Output_Get:double)
   n21 = Get PawnVelocityZ()
   > Branch(Condition=AND Boolean(A=@n20, B="float > float"(A=@n21, B=100)))
   Set capeAlpha(execute=@n19.then_1, capeAlpha:double=float - float(A=1.1, B="To Float (Boolean)"(InBool=@n17)) {is=title K2Node_CallFunction}) -> (Output_Get:double)
   > Set isWearingCape(isWearingCape=@n18)
   > n22 = Sequence()
   > n23 = Branch(Condition=IsInVehicle!(self:Fort Player Pawn Object Reference=Get Pawn Owner() -> (Pawn Owner:Player Pawn Athena Object Reference)) -> (ReturnValue:bool))
   n24 = GetPartSkeletalMeshComponent()
   n25 = IsInVehicle(self=Get Pawn Owner())
   > Set ClothBlendWeight!(execute, ClothBlendWeight:float="To Float (Boolean)"(InBool=NOT Boolean(A=@n25)), self:Skeletal Mesh Component Object Reference=@n24) -> (then, Output_Get:float)
   Set ClothBlendWeight(execute=@n23.else, ClothBlendWeight=Get ClothBlendDelta(), self=@n24)
   n26 = Get PawnVelocityZ()
   n27 = Get bIsSkyDiving()
   > n28 = Branch(Condition=AND Boolean(A="float < float"(A=@n26, B=-700), B=NOT Boolean(A=@n27)))
   > Set ClothBlendWeight(ClothBlendWeight=Get ClothBlendDeltaFall(), self=@n24)
   n29 = float * float(A="To Float (Boolean)"(InBool=Get bIsParachuteOpened()), B=0.4)
   Set ClothBlendWeight(
    execute=@n28.else
    ClothBlendWeight=float * float(A=Get ClothBlendWeight!(self:Skeletal Mesh Component Object Reference=@n24) -> (ClothBlendWeight:float), B=float - float(B=@n29))
    self=@n24
   )
   n30 = Get bIsSkyDiving()
   n31 = Branch(execute=@n22.then_1, Condition=AND Boolean(A=Get bIsJumping(), B=NOT Boolean(A=@n30)))
   > Delay(Duration=1.1)
   > ForceClothNextUpdateTeleportAndReset(self=Get MeshToCopy_0())
   n32 = Get PawnVelocityZ()
   n33 = "float < float"(A=@n32, B=50)
   n34 = "float > float"(A=@n32, B=-50)
   Branch(execute=@n31.else, Condition=AND Boolean(A=@n30, B=AND Boolean(A=@n33, B=@n34)))
   Event BlueprintInitializeAnimation()
   > Set MeshToCopy_0(MeshToCopy_0:Skeletal Mesh Component Object Reference=GetPartSkeletalMeshComponent()) -> (Output_Get:Skeletal Mesh Component Object Reference)
   > Set random_stream(random_stream:Random Stream Structure=MakeRandomStream!(InitialSeed:int=0) -> (ReturnValue:Random Stream Structure)) -> (Output_Get:Random Stream Structure)
   > n35 = Cast To ClothingSimulationInteractorNv!(
    execute
    Object=GetClothingSimulationInteractor!(self:Skeletal Mesh Component Object Reference=GetPartSkeletalMeshComponent(), ClothingAssetName:Name=None) -> (ReturnValue:Clothing Simulation Interactor Object Reference)
   ) -> (then, CastFailed, AsClothing Simulation Interactor Nv:Clothing Simulation Interactor Nv Object Reference) {is=title K2Node_DynamicCast}
   > Set FireElf_clothing_sim_interactor(FireElf_clothing_sim_interactor:Clothing Simulation Interactor Nv Object Reference=@n35) -> (Output_Get:Clothing Simulation Interactor Nv Object Reference)
   n36 = TryGetPawnOwner()
   > Is Valid(InputObject=@n36)
   > n37 = Cast To PlayerPawn_Athena(Object=@n36)
   > Set Pawn Owner(Pawn Owner:Player Pawn Athena Object Reference=@n37) -> (Output_Get:Player Pawn Athena Object Reference)
   Event OnEndCrouching()
   > Delay(Duration=0.125)
   > ForceClothNextUpdateTeleportAndReset(self=Get MeshToCopy_0())
   TryGetPawnOwner()
   "Not Equal (Object)"!(
    A:Object Reference=FindChildCommponentFromAttachSocket!(
     self:Custom Character Part Anim Instance Object Reference
     Parent:Skeletal Mesh Component Object Reference=@n24
     bIncludeAllDescendants:bool=false
     AttachSocket:Name=Cape
    ) -> (ReturnValue:Skeletal Mesh Component Object Reference)
    B:Object Reference
   ) -> (ReturnValue:bool)
   EnableGravityOverride!(
    execute
    self:Clothing Simulation Interactor Object Reference=Get FireElf_clothing_sim_interactor() -> (FireElf_clothing_sim_interactor:Clothing Simulation Interactor Nv Object Reference)
    InVector:"Vector (by ref)"=vector * float(A="0,0,500", B=Get PawnSpeed2D())
   ) -> (then)
   RandomFloatInRangeFromStream(Stream=Get random_stream(), Min=1, Max=100)
   vector * vector(A=@n15, B="-5,-5,-5")
   vector + vector(A:Vector=vector * vector(A=@n15, B="100,100,120"), B:Vector="0, 0,2000") -> (ReturnValue:Vector) {is=title K2Node_CommutativeAssociativeBinaryOperator}
   Event OnBeginCrouching()
   Set random_stream(random_stream=MakeRandomStream(InitialSeed=0))
   n38 = RandomFloatInRangeFromStream(Stream=Get random_stream(), Min=-20000, Max=20000)
   Set skydiveVectorA(skydiveVectorA:Vector=MakeVector(X=@n38, Y=@n38)) -> (Output_Get:Vector)
  ]}
]
variables [
 capeAlpha {type=number}
 ClothBlendDelta {type=number}
 ClothBlendDeltaFall {type=number, default=1}
 FireElf_clothing_sim_interactor {type=ClothingSimulationInteractorNv, default=null}
 IsWearingBackbling {type=boolean}
 isWearingCape {type=boolean}
 MeshToCopy_0 {type=SkeletalMeshComponent}
 Pawn Owner {type=PlayerPawn_Athena_C}
 random_stream {type=RandomStream, default {initialSeed=0, seed=0}}
 Rigid Body Not Enabled {type=boolean}
 SkydiveAlpha {type=number}
 skydiveVectorA {type=Vector}
 skydiving vector force {type=Vector}
]
