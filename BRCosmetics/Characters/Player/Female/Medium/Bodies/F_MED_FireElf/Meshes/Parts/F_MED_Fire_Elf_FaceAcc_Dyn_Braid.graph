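# Physics asset for the dynamic back braid on the medium female Fire Elf's face accessory mesh.
# Six capsule bodies form a braid chain constrained from the head, with swing and twist limits tight near the root and looser toward the tip.
# Capsules on the upper spine, clavicles, neck and head approximate the body around the braid, joined by constraints with free swing and twist.
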PhysicsAsset /BRCosmetics/Characters/Player/Female/Medium/Bodies/F_MED_FireElf/Meshes/Parts/F_MED_Fire_Elf_FaceAcc_Dyn_Braid
bodies {
 clavicle_l [{shapeType=Capsule, length=7.5312, radius=5.5988, center=<4.1434 -0.4322 -0.1613>, rotation=<rot 0.4354 -94.9807 -85.0169>}]
 clavicle_r [{shapeType=Capsule, length=7.5312, radius=5.5988, center=<-4.126 -0.0611 0.855>, rotation=<rot 0.4354 -94.9807 -85.0169>}]
 dyn_backBraid_01 [{shapeType=Capsule, length=1.625, radius=1.75, center=<4.3836 0 0.0001>, rotation=<rot 90 0 0>}]
 dyn_backBraid_02 [{shapeType=Capsule, length=3.3125, radius=1.6875, center=<3 0 0>, rotation=<rot 90 0 0>}]
 dyn_backBraid_03 [{shapeType=Capsule, length=3.2188, radius=1.7812, center=<3 0 0>, rotation=<rot 90 0 0>}]
 dyn_backBraid_04 [{shapeType=Capsule, length=3.6562, radius=1.3438, center=<3 0 0>, rotation=<rot 90 0 0>}]
 dyn_backBraid_05 [{shapeType=Capsule, length=3.8438, radius=1.1562, center=<3 0 0>, rotation=<rot 90 0 0>}]
 dyn_backBraid_06 [{shapeType=Capsule, length=3.5, radius=1.5, center=<3 0 0>, rotation=<rot 90 0 0>}]
 head [{shapeType=Capsule, length=3.0739, radius=5.5391, center=<1.5569 -0.3138 -0.0532>, rotation=<rot -0.3094 99.2805 -89.484>}]
 neck_01 [{shapeType=Capsule, length=0.01, radius=4.4254, rotation=<rot 0 -90 0>}]
 neck_02 [{shapeType=Capsule, length=2.4162, radius=4.0192, rotation=<rot 0 -90 0>}]
 spine_04 [{shapeType=Capsule, length=8.6875, radius=9.4688, center=<4.6506 -0.736 0>}]
 spine_05 [{shapeType=Capsule, length=8.6875, radius=9.4688, center=<2.9378 0.3311 0>}]
}
constraints [
 {
  bone1Name=clavicle_l
  bone2Name=spine_05
  swing1LimitDegrees=45
  swing1Motion=Free
  swing2LimitDegrees=45
  swing2Motion=Free
  twistLimitDegrees=45
  twistMotion=Free
 }
 {
  bone1Name=clavicle_r
  bone2Name=spine_05
  swing1LimitDegrees=45
  swing1Motion=Free
  swing2LimitDegrees=45
  swing2Motion=Free
  twistLimitDegrees=45
  twistMotion=Free
 }
 {
  bone1Name=neck_02
  bone2Name=neck_01
  swing1LimitDegrees=45
  swing1Motion=Free
  swing2LimitDegrees=45
  swing2Motion=Free
  twistLimitDegrees=45
  twistMotion=Free
 }
 {bone1Name=dyn_backBraid_01, bone2Name=head, swing1LimitDegrees=40, swing2LimitDegrees=10, twistLimitDegrees=2}
 {bone1Name=dyn_backBraid_02, bone2Name=dyn_backBraid_01, swing1LimitDegrees=20, swing2LimitDegrees=20, twistLimitDegrees=4}
 {bone1Name=dyn_backBraid_03, bone2Name=dyn_backBraid_02, swing1LimitDegrees=35, swing2LimitDegrees=35, twistLimitDegrees=10}
 {bone1Name=dyn_backBraid_04, bone2Name=dyn_backBraid_03, swing1LimitDegrees=45, swing2LimitDegrees=45, twistLimitDegrees=10}
 {bone1Name=dyn_backBraid_05, bone2Name=dyn_backBraid_04, swing1LimitDegrees=45, swing2LimitDegrees=45, twistLimitDegrees=10}
 {bone1Name=dyn_backBraid_06, bone2Name=dyn_backBraid_05, swing1LimitDegrees=45, swing2LimitDegrees=45, twistLimitDegrees=10}
 {
  bone1Name=neck_01
  bone2Name=spine_05
  swing1LimitDegrees=45
  swing1Motion=Free
  swing2LimitDegrees=45
  swing2Motion=Free
  twistLimitDegrees=45
  twistMotion=Free
 }
 {
  bone1Name=head
  bone2Name=neck_02
  swing1LimitDegrees=45
  swing1Motion=Free
  swing2LimitDegrees=45
  swing2Motion=Free
  twistLimitDegrees=45
  twistMotion=Free
 }
 {
  bone1Name=spine_05
  bone2Name=spine_04
  swing1LimitDegrees=45
  swing1Motion=Free
  swing2LimitDegrees=45
  swing2Motion=Free
  twistLimitDegrees=45
  twistMotion=Free
 }
]
