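# Physics asset for the temple braids on the Fire Elf female medium character's face part, letting both braids swing as secondary motion.
# Each side is a five-capsule chain from a base bone on the head, with swing and twist limits mostly loosening toward the tip.
# Head, neck, clavicle and upper spine shapes, joined by free-rotation constraints, supply body volumes, likely for the braids to collide against.
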PhysicsAsset /BRCosmetics/Characters/Player/Female/Medium/Bodies/F_MED_FireElf/Meshes/Parts/F_MED_Fire_Elf_Face_TempleBraid_Physics
bodies {
 clavicle_l [{shapeType=Capsule, length=7.7812, radius=5.3488, center=<4.1434 -0.4322 -0.1613>, rotation=<rot 0.4354 -94.9807 -85.0169>}]
 clavicle_r [{shapeType=Capsule, length=7.7812, radius=5.3488, center=<-4.126 -0.0611 0.855>, rotation=<rot 0.4354 -94.9807 -85.0169>}]
 dyn_templeBraid_01_l [{
   shapeType=Capsule
   length=1.482
   radius=1.5558
   shapeName=dyn_templeBraid_01_l_capsule
   center=<-2.9924 0.0448 -0.1131>
   rotation=<rot 3.8824 -90 -71.5419>
  }]
 dyn_templeBraid_01_r [{
   shapeType=Capsule
   length=1.9821
   radius=1.0558
   shapeName=dyn_templeBraid_01_r_capsule
   center=<2.6566 -0.0372 0.2246>
   rotation=<rot 3.8817 -90 88.4528>
  }]
 dyn_templeBraid_02_l [{
   shapeType=Capsule
   length=2.3412
   radius=1.9989
   shapeName=dyn_templeBraid_02_l_capsule
   center=<-1.6527 0.0566 -0.1145>
   rotation=<rot 0 -89.1378 -75.305>
  }]
 dyn_templeBraid_02_r [{
   shapeType=Capsule
   length=2.3416
   radius=1.9971
   shapeName=dyn_templeBraid_02_r_capsule
   center=<1.652 -0.0564 0.1122>
   rotation=<rot 0 -89.1313 -75.2922>
  }]
 dyn_templeBraid_03_l [{
   shapeType=Capsule
   length=2.2154
   radius=2.2879
   shapeName=dyn_templeBraid_03_l_capsule
   center=<-1.8919 -0.0376 -0.0818>
   rotation=<rot 0 -89.772 -75.2725>
  }]
 dyn_templeBraid_03_r [{
   shapeType=Capsule
   length=2.4676
   radius=2.0386
   shapeName=dyn_templeBraid_03_r_capsule
   center=<1.8913 0.0339 0.0734>
   rotation=<rot 0 -89.8333 -75.1658>
  }]
 dyn_templeBraid_04_l [{
   shapeType=Capsule
   length=3.1702
   radius=1.5533
   shapeName=dyn_templeBraid_04_l_capsule
   center=<-2.5818 -0.0501 -0.1676>
   rotation=<rot 0 -93.859 -88.9469>
  }]
 dyn_templeBraid_04_r [{
   shapeType=Capsule
   length=3.4413
   radius=1.5498
   shapeName=dyn_templeBraid_04_r_capsule
   center=<2.8725 0.0807 0.1611>
   rotation=<rot 0 -92.6965 -89.2168>
  }]
 dyn_templeBraid_Base_l [{
   shapeType=Capsule
   length=2.4697
   radius=1.529
   shapeName=dyn_templeBraid_Base_l_capsule
   center=<-2.2692 -0.0597 0.2049>
   rotation=<rot 5.5837 -90.0001 95.4756>
  }]
 dyn_templeBraid_Base_r [{
   shapeType=Capsule
   length=2.4665
   radius=1.569
   shapeName=dyn_templeBraid_Base_r_capsule
   center=<2.2414 0.0608 -0.1792>
   rotation=<rot 5.1154 -90.0001 96.5051>
  }]
 head [{shapeType=Capsule, length=12.4489, radius=5.5391, center=<1.5569 -0.3138 -0.0532>, rotation=<rot -0.3094 99.2805 -89.484>}]
 neck_01 [{shapeType=Capsule, length=0.01, radius=4.4254, rotation=<rot 0 -90 0>}]
 neck_02 [{shapeType=Capsule, length=0.01, radius=4.4254, rotation=<rot 0 -90 0>}]
 spine_04 [
  {shapeType=Sphere, radius=5.77, center=<3.0921 6.2805 -6.6159>}
  {shapeType=Sphere, radius=5.7711, center=<3.092 6.1711 6.4423>}
  {shapeType=Capsule, length=5.4692, radius=11.4058, center=<1.2907 1.4182 0>}
 ]
 spine_05 [{shapeType=Capsule, length=6.9062, radius=11.25, center=<-0.3385 -2.1504 0>}]
}
constraints [
 {
  bone1Name=head
  bone2Name=neck_02
  swing1LimitDegrees=45
  swing1Motion=Free
  swing2LimitDegrees=45
  swing2Motion=Free
  twistLimitDegrees=45
  twistMotion=Free
 }
 {
  bone1Name=neck_02
  bone2Name=neck_01
  swing1LimitDegrees=45
  swing1Motion=Free
  swing2LimitDegrees=45
  swing2Motion=Free
  twistLimitDegrees=45
  twistMotion=Free
 }
 {bone1Name=dyn_templeBraid_01_l, bone2Name=dyn_templeBraid_Base_l, swing1LimitDegrees=15, swing2LimitDegrees=50, twistLimitDegrees=15}
 {bone1Name=dyn_templeBraid_02_l, bone2Name=dyn_templeBraid_01_l, swing1LimitDegrees=20, swing2LimitDegrees=20, twistLimitDegrees=10}
 {bone1Name=dyn_templeBraid_03_l, bone2Name=dyn_templeBraid_02_l, swing1LimitDegrees=30, swing2LimitDegrees=30, twistLimitDegrees=15}
 {bone1Name=dyn_templeBraid_04_l, bone2Name=dyn_templeBraid_03_l, swing1LimitDegrees=35, swing2LimitDegrees=35, twistLimitDegrees=35}
 {bone1Name=dyn_templeBraid_01_r, bone2Name=dyn_templeBraid_Base_r, swing1LimitDegrees=15, swing2LimitDegrees=50, twistLimitDegrees=15}
 {bone1Name=dyn_templeBraid_02_r, bone2Name=dyn_templeBraid_01_r, swing1LimitDegrees=20, swing2LimitDegrees=20, twistLimitDegrees=10}
 {bone1Name=dyn_templeBraid_03_r, bone2Name=dyn_templeBraid_02_r, swing1LimitDegrees=30, swing2LimitDegrees=30, twistLimitDegrees=15}
 {bone1Name=dyn_templeBraid_04_r, bone2Name=dyn_templeBraid_03_r, swing1LimitDegrees=35, swing2LimitDegrees=35, twistLimitDegrees=35}
 {
  bone1Name=neck_01
  bone2Name=spine_05
  swing1LimitDegrees=45
  swing1Motion=Free
  swing2LimitDegrees=45
  swing2Motion=Free
  twistLimitDegrees=45
  twistMotion=Free
 }
 {
  bone1Name=clavicle_l
  bone2Name=spine_05
  swing1LimitDegrees=45
  swing1Motion=Free
  swing2LimitDegrees=45
  swing2Motion=Free
  twistLimitDegrees=45
  twistMotion=Free
 }
 {
  bone1Name=clavicle_r
  bone2Name=spine_05
  swing1LimitDegrees=45
  swing1Motion=Free
  swing2LimitDegrees=45
  swing2Motion=Free
  twistLimitDegrees=45
  twistMotion=Free
 }
 {
  bone1Name=spine_05
  bone2Name=spine_04
  swing1LimitDegrees=45
  swing1Motion=Free
  swing2LimitDegrees=45
  swing2Motion=Free
  twistLimitDegrees=45
  twistMotion=Free
 }
 {bone1Name=dyn_templeBraid_Base_r, bone2Name=head, swing1LimitDegrees=20, swing2LimitDegrees=20, twistLimitDegrees=2}
 {bone1Name=dyn_templeBraid_Base_l, bone2Name=head, swing1LimitDegrees=20, swing2LimitDegrees=20, twistLimitDegrees=2}
]
