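# Anim Blueprint for the Fire Elf face accessory, adding secondary physics motion to her braids.
# The output path copies the attached parent's pose, then runs two chained RigidBody nodes with override physics assets, the first fed skydiving external force.
# AnimDynamics, Flap and bone-transform chains for temple and back braids remain unconnected to the output.
# The event graph computes cape and head-to-spine alphas used only by those unconnected chains.
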AnimBlueprint /BRCosmetics/Characters/Player/Female/Medium/Bodies/F_MED_FireElf/Meshes/Parts/F_MED_Fire_Elf_FaceAcc_Skeleton_AnimBlueprint
target_skeleton=&/F_MED_Fire_Elf_FaceAcc_Skeleton
graphs [
 AnimGraph {nodes [
   n1 = Anim Dynamics - Bone: dyn_templeBraid_Base_r(Alpha=1, ExternalForce="0,0,0") {
    angularDampingOverride=0.999
    bOverrideAngularDamping=true
    bUseGravityOverride=true
    bUsePlanarLimit=false
    bUseSphericalLimits=true
    simulationSpace=RootRelative
    alphaBoolBlend {blendOption=Linear}
    boundBone {boneName=dyn_templeBraid_Base_r}
    componentLinearAccScale=<0.5 0.5 0.5>
    componentLinearVelScale=<0.5 0.5 0.5>
    gravityOverride {z=-980}
    physicsBodyDefinitions [{
      boundBone {boneName=dyn_templeBraid_Base_r}
      boxExtents=<2 20 20>
      constraintSetup {angularLimitsMax=<5 10 5>, angularLimitsMin=<-5 -10 -5>, angularTarget=<0 0 0>}
      localJointOffset=<5 0 0>
     }]
   }
   n2 = Anim Dynamics - Bone: dyn_templeBraid_Base_l(ComponentPose=@n1, Alpha=1, ExternalForce="0,0,0") {
    angularDampingOverride=0.999
    bOverrideAngularDamping=true
    bUseGravityOverride=true
    bUsePlanarLimit=false
    bUseSphericalLimits=true
    simulationSpace=RootRelative
    alphaBoolBlend {blendOption=Linear}
    boundBone {boneName=dyn_templeBraid_Base_l}
    componentLinearAccScale=<0.5 0.5 0.5>
    componentLinearVelScale=<0.5 0.5 0.5>
    gravityOverride {z=-980}
    physicsBodyDefinitions [{
      boundBone {boneName=dyn_templeBraid_Base_l}
      boxExtents=<2 20 20>
      constraintSetup {angularLimitsMax=<5 10 5>, angularLimitsMin=<-5 -10 -5>, angularTarget=<0 0 0>}
      localJointOffset=<-5 0 0>
     }]
   }
   n3 = Anim Dynamics - Bone: dyn_templeBraid_01_r(ComponentPose=@n2, Alpha=1, ExternalForce="0,0,0") {
    angularDampingOverride=0.999
    bOverrideAngularDamping=true
    bUseGravityOverride=true
    bUsePlanarLimit=false
    bUseSphericalLimits=true
    simulationSpace=RootRelative
    alphaBoolBlend {blendOption=Linear}
    boundBone {boneName=dyn_templeBraid_01_r}
    componentLinearAccScale=<0.5 0.5 0.5>
    componentLinearVelScale=<0.5 0.5 0.5>
    gravityOverride {z=-980}
    physicsBodyDefinitions [{
      boundBone {boneName=dyn_templeBraid_01_r}
      boxExtents=<2 20 20>
      constraintSetup {angularLimitsMax=<5 10 5>, angularLimitsMin=<-5 -10 -5>, angularTarget=<0 0 0>}
      localJointOffset=<5 0 0>
     }]
   }
   n4 = Anim Dynamics - Bone: dyn_templeBraid_01_l(ComponentPose=@n3, Alpha=1, ExternalForce="0,0,0") {
    angularDampingOverride=0.999
    bOverrideAngularDamping=true
    bUseGravityOverride=true
    bUsePlanarLimit=false
    bUseSphericalLimits=true
    simulationSpace=RootRelative
    alphaBoolBlend {blendOption=Linear}
    boundBone {boneName=dyn_templeBraid_01_l}
    componentLinearAccScale=<0.5 0.5 0.5>
    componentLinearVelScale=<0.5 0.5 0.5>
    gravityOverride {z=-980}
    physicsBodyDefinitions [{
      boundBone {boneName=dyn_templeBraid_01_l}
      boxExtents=<2 20 20>
      constraintSetup {angularLimitsMax=<5 10 5>, angularLimitsMin=<-5 -10 -5>, angularTarget=<0 0 0>}
      localJointOffset=<-5 0 0>
     }]
   }
   n5 = Anim Dynamics - Bone: dyn_templeBraid_02_r(ComponentPose=@n4, Alpha=1, ExternalForce="0,0,0") {
    angularDampingOverride=0.999
    bOverrideAngularDamping=true
    bUseGravityOverride=true
    bUsePlanarLimit=false
    bUseSphericalLimits=true
    simulationSpace=RootRelative
    alphaBoolBlend {blendOption=Linear}
    boundBone {boneName=dyn_templeBraid_02_r}
    componentLinearAccScale=<0.5 0.5 0.5>
    componentLinearVelScale=<0.5 0.5 0.5>
    gravityOverride {z=-980}
    physicsBodyDefinitions [{
      boundBone {boneName=dyn_templeBraid_02_r}
      boxExtents=<2 20 20>
      constraintSetup {angularLimitsMax=<5 30 5>, angularLimitsMin=<-5 -30 -5>, angularTarget=<0 0 0>}
      localJointOffset=<5 0 0>
     }]
    sphericalLimits [
     {limitRadius=13, drivingBone {boneName=spine_05}, sphereLocalOffset=<0 0 0>}
     {limitRadius=5.5, drivingBone {boneName=clavicle_l}, sphereLocalOffset=<10 0 0>}
     {limitRadius=5.5, drivingBone {boneName=clavicle_r}, sphereLocalOffset=<-5 0 0>}
    ]
   }
   n6 = Anim Dynamics - Bone: dyn_templeBraid_02_l(ComponentPose=@n5, Alpha=1, ExternalForce="0,0,0") {
    angularDampingOverride=0.999
    bOverrideAngularDamping=true
    bUseGravityOverride=true
    bUsePlanarLimit=false
    bUseSphericalLimits=true
    simulationSpace=RootRelative
    alphaBoolBlend {blendOption=Linear}
    boundBone {boneName=dyn_templeBraid_02_l}
    componentLinearAccScale=<0.5 0.5 0.5>
    componentLinearVelScale=<0.5 0.5 0.5>
    gravityOverride {z=-980}
    physicsBodyDefinitions [{
      boundBone {boneName=dyn_templeBraid_02_l}
      boxExtents=<2 20 20>
      constraintSetup {angularLimitsMax=<5 30 5>, angularLimitsMin=<-5 -30 -5>, angularTarget=<0 0 0>}
      localJointOffset=<-5 0 0>
     }]
    sphericalLimits [
     {limitRadius=13, drivingBone {boneName=spine_05}, sphereLocalOffset=<0 0 0>}
     {limitRadius=5.5, drivingBone {boneName=clavicle_l}, sphereLocalOffset=<10 0 0>}
     {limitRadius=5.5, drivingBone {boneName=clavicle_r}, sphereLocalOffset=<0 0 0>}
    ]
   }
   n7 = Anim Dynamics - Bone: dyn_templeBraid_03_r(ComponentPose=@n6, Alpha=1, ExternalForce="0,0,0") {
    angularDampingOverride=0.999
    bOverrideAngularDamping=true
    bUseGravityOverride=true
    bUsePlanarLimit=false
    bUseSphericalLimits=true
    simulationSpace=RootRelative
    alphaBoolBlend {blendOption=Linear}
    boundBone {boneName=dyn_templeBraid_03_r}
    componentLinearAccScale=<0.5 0.5 0.5>
    componentLinearVelScale=<0.5 0.5 0.5>
    gravityOverride {z=-980}
    physicsBodyDefinitions [{
      boundBone {boneName=dyn_templeBraid_03_r}
      boxExtents=<2 20 20>
      constraintSetup {angularLimitsMax=<5 30 5>, angularLimitsMin=<-5 -30 -5>, angularTarget=<0 0 0>}
      localJointOffset=<5 0 0>
     }]
    sphericalLimits [
     {limitRadius=13, drivingBone {boneName=spine_05}, sphereLocalOffset=<0 0 0>}
     {limitRadius=5.5, drivingBone {boneName=clavicle_l}, sphereLocalOffset=<10 0 0>}
     {limitRadius=5.5, drivingBone {boneName=clavicle_r}, sphereLocalOffset=<-5 0 0>}
    ]
   }
   n8 = Anim Dynamics - Bone: dyn_templeBraid_03_l(ComponentPose=@n7, Alpha=1, ExternalForce="0,0,0") {
    angularDampingOverride=0.999
    bOverrideAngularDamping=true
    bUseGravityOverride=true
    bUsePlanarLimit=false
    bUseSphericalLimits=true
    simulationSpace=RootRelative
    alphaBoolBlend {blendOption=Linear}
    boundBone {boneName=dyn_templeBraid_03_l}
    componentLinearAccScale=<0.5 0.5 0.5>
    componentLinearVelScale=<0.5 0.5 0.5>
    gravityOverride {z=-980}
    physicsBodyDefinitions [{
      boundBone {boneName=dyn_templeBraid_03_l}
      boxExtents=<2 20 20>
      constraintSetup {angularLimitsMax=<5 30 5>, angularLimitsMin=<-5 -30 -5>, angularTarget=<0 0 0>}
      localJointOffset=<-5 0 0>
     }]
    sphericalLimits [
     {limitRadius=13, drivingBone {boneName=spine_05}, sphereLocalOffset=<0 0 0>}
     {limitRadius=5.5, drivingBone {boneName=clavicle_l}, sphereLocalOffset=<10 0 0>}
     {limitRadius=5.5, drivingBone {boneName=clavicle_r}, sphereLocalOffset=<0 0 0>}
    ]
   }
   n9 = Anim Dynamics - Bone: dyn_templeBraid_04_r(ComponentPose=@n8, Alpha=1, ExternalForce="0,0,0") {
    angularDampingOverride=0.999
    bOverrideAngularDamping=true
    bUseGravityOverride=true
    bUsePlanarLimit=false
    bUseSphericalLimits=true
    simulationSpace=RootRelative
    alphaBoolBlend {blendOption=Linear}
    boundBone {boneName=dyn_templeBraid_04_r}
    componentLinearAccScale=<0.5 0.5 0.5>
    componentLinearVelScale=<0.5 0.5 0.5>
    gravityOverride {z=-980}
    physicsBodyDefinitions [{
      boundBone {boneName=dyn_templeBraid_04_r}
      boxExtents=<2 20 20>
      constraintSetup {angularLimitsMax=<5 30 5>, angularLimitsMin=<-5 -30 -5>, angularTarget=<0 0 0>}
      localJointOffset=<5 0 0>
     }]
    sphericalLimits [
     {limitRadius=13, drivingBone {boneName=spine_05}, sphereLocalOffset=<0 0 0>}
     {limitRadius=5.5, drivingBone {boneName=clavicle_l}, sphereLocalOffset=<10 0 0>}
     {limitRadius=5.5, drivingBone {boneName=clavicle_r}, sphereLocalOffset=<-5 0 0>}
    ]
   }
   Anim Dynamics - Bone: dyn_templeBraid_04_l(ComponentPose=@n9, Alpha=1, ExternalForce="0,0,0") {
    angularDampingOverride=0.999
    bOverrideAngularDamping=true
    bUseGravityOverride=true
    bUsePlanarLimit=false
    bUseSphericalLimits=true
    simulationSpace=RootRelative
    alphaBoolBlend {blendOption=Linear}
    boundBone {boneName=dyn_templeBraid_04_l}
    componentLinearAccScale=<0.5 0.5 0.5>
    componentLinearVelScale=<0.5 0.5 0.5>
    gravityOverride {z=-980}
    physicsBodyDefinitions [{
      boundBone {boneName=dyn_templeBraid_04_l}
      boxExtents=<2 20 20>
      constraintSetup {angularLimitsMax=<5 30 5>, angularLimitsMin=<-5 -30 -5>, angularTarget=<0 0 0>}
      localJointOffset=<-5 0 0>
     }]
    sphericalLimits [
     {limitRadius=13, drivingBone {boneName=spine_05}, sphereLocalOffset=<0 0 0>}
     {limitRadius=5.5, drivingBone {boneName=clavicle_l}, sphereLocalOffset=<10 0 0>}
     {limitRadius=5.5, drivingBone {boneName=clavicle_r}, sphereLocalOffset=<0 0 0>}
    ]
   }
   n10 = Flap Node - Bone: dyn_backBraid_01!(ComponentPose:Component Space Pose Link Structure, Alpha:float=1) -> (Pose:Component Space Pose Link Structure) {
    is=title FortAnimGraphNode_Flap
    bUseClamp=true
    targetClampMax=2
    targetClampMin=-1
    targetComponent=TranslationY
    alphaBoolBlend {blendOption=Linear}
    sourceDrives [{
      bUseRange=false
      drivingCurve=null
      multiplier=-1
      rangeMax=1
      rangeMin=0
      remappedMax=1
      remappedMin=0
      sourceComponent=RotationZ
      useQuaternion=true
      sourceBone {boneName=head}
     }]
    targetBone {boneName=dyn_backBraid_01}
   }
   n11 = Flap Node - Bone: dyn_backBraid_01(ComponentPose=@n10, Alpha=1) {
    bUseClamp=true
    targetClampMax=100
    targetComponent=TranslationX
    alphaBoolBlend {blendOption=Linear}
    sourceDrives [{
      bUseRange=false
      drivingCurve=null
      multiplier=5
      rangeMax=50
      rangeMin=0
      remappedMax=80
      remappedMin=0
      sourceComponent=RotationZ
      useQuaternion=true
      sourceBone {boneName=head}
     }]
    targetBone {boneName=dyn_backBraid_01}
   }
   n12 = Flap Node - Bone: dyn_backBraid_01(ComponentPose=@n11, Alpha=1) {
    bUseClamp=true
    targetClampMin=-1
    targetComponent=RotationY
    alphaBoolBlend {blendOption=Linear}
    sourceDrives [{
      bUseRange=false
      drivingCurve=null
      multiplier=-0.25
      rangeMax=50
      rangeMin=0
      remappedMax=50
      remappedMin=0
      sourceComponent=RotationZ
      useQuaternion=true
      sourceBone {boneName=head}
     }]
    targetBone {boneName=dyn_backBraid_01}
   }
   n13 = Flap Node - Bone: dyn_backBraid_01(ComponentPose=@n12, Alpha=1) {
    bUseClamp=true
    targetClampMax=5
    targetClampMin=-5
    targetComponent=RotationZ
    alphaBoolBlend {blendOption=Linear}
    sourceDrives [{
      bUseRange=false
      drivingCurve=null
      multiplier=1
      rangeMax=40
      rangeMin=0
      remappedMax=80
      remappedMin=0
      sourceComponent=RotationZ
      useQuaternion=true
      sourceBone {boneName=head}
     }]
    targetBone {boneName=dyn_backBraid_01}
   }
   n14 = Get capeAlpha() -> (capeAlpha:double)
   n15 = Anim Dynamics - Bone: dyn_backBraid_02!(ComponentPose=@n13, Alpha=@n14, AngularSpringConstant:float=4000, ExternalForce="0,0,0") -> (Pose) {
    angularDampingOverride=0.9999
    angularSpringConstant=1000
    bAngularSpring=true
    bGravityOverrideInSimSpace=true
    bOverrideAngularDamping=true
    bUseSphericalLimits=true
    gravityScale=0.35
    simulationSpace=RootRelative
    alphaBoolBlend {blendOption=Linear}
    alphaScaleBiasClamp {bClampResult=true, clampMin=0.1}
    boundBone {boneName=dyn_backBraid_02}
    componentAppliedLinearAccClamp=<70 70 70>
    componentLinearVelScale=<300 300 300>
    gravityOverride {z=-980}
    physicsBodyDefinitions [{
      collisionType=CustomSphere
      sphereCollisionRadius=3
      boundBone {boneName=dyn_backBraid_02}
      boxExtents=<3 20 20>
      constraintSetup {
       angularLimitsMax=<5 10 20>
       angularLimitsMin=<-5 -10 -40>
       angularTarget=<1 0 0>
       linearAxesMax=<0.05 0.05 0.05>
       linearAxesMin=<-0.05 -0.05 -0.05>
      }
      localJointOffset=<3 0 0>
     }]
    planarLimits [
     {drivingBone {boneName=spine_04}, planeTransform {location=<0 -10 0>, rotation=<rot 0 20 -90>}}
     {drivingBone {boneName=clavicle_l}, planeTransform {location=<0 0 0>, rotation=<rot 0 -20 -90>}}
     {drivingBone {boneName=clavicle_r}, planeTransform {location=<0 0 0>, rotation=<rot 0 -20 90>}}
    ]
    sphericalLimits [
     {limitRadius=10, drivingBone {boneName=spine_05}, sphereLocalOffset=<0 0 0>}
     {limitRadius=6.5, drivingBone {boneName=clavicle_r}, sphereLocalOffset=<0 0 0>}
     {limitRadius=6.5, drivingBone {boneName=clavicle_l}, sphereLocalOffset=<0 0 0>}
    ]
   }
   n16 = Anim Dynamics - Bone: dyn_backBraid_03!(ComponentPose=@n15, Alpha=@n14, AngularSpringConstant:float=4000, ExternalForce="0,0,0") -> (Pose) {
    angularDampingOverride=0.9999
    angularSpringConstant=1000
    bAngularSpring=true
    bGravityOverrideInSimSpace=true
    bOverrideAngularDamping=true
    gravityScale=0.35
    simulationSpace=RootRelative
    alphaBoolBlend {blendOption=Linear}
    alphaScaleBiasClamp {bClampResult=true}
    boundBone {boneName=dyn_backBraid_03}
    componentAppliedLinearAccClamp=<70 70 70>
    componentLinearVelScale=<300 300 300>
    gravityOverride {z=-980}
    physicsBodyDefinitions [{
      collisionType=CustomSphere
      sphereCollisionRadius=3
      boundBone {boneName=dyn_backBraid_03}
      boxExtents=<3 20 20>
      constraintSetup {
       angularLimitsMax=<5 10 10>
       angularLimitsMin=<-5 -10 -40>
       angularTarget=<1 0 0>
       linearAxesMax=<0.25 0.25 0.25>
       linearAxesMin=<-0.25 -0.25 -0.25>
      }
      localJointOffset=<3 0 0>
     }]
    planarLimits [{drivingBone {boneName=spine_04}, planeTransform {location=<0 -7.5 0>, rotation=<rot 0 5.0001 -89.9999>}}]
    sphericalLimits [{limitRadius=10.5, drivingBone {boneName=spine_05}, sphereLocalOffset=<0 0 0>}]
   }
   n17 = Anim Dynamics - Bone: dyn_backBraid_05!(ComponentPose=@n16, Alpha=@n14, AngularSpringConstant:float=4000, ExternalForce="0,0,0") -> (Pose) {
    angularDampingOverride=0.9999
    angularSpringConstant=1000
    bAngularSpring=true
    bGravityOverrideInSimSpace=true
    bOverrideAngularDamping=true
    gravityScale=0.35
    simulationSpace=RootRelative
    alphaBoolBlend {blendOption=Linear}
    alphaScaleBiasClamp {bClampResult=true}
    boundBone {boneName=dyn_backBraid_05}
    componentAppliedLinearAccClamp=<70 70 70>
    componentLinearVelScale=<300 300 300>
    gravityOverride {z=-980}
    physicsBodyDefinitions [{
      collisionType=CustomSphere
      sphereCollisionRadius=3
      boundBone {boneName=dyn_backBraid_05}
      boxExtents=<3 20 20>
      constraintSetup {
       angularLimitsMax=<5 10 10>
       angularLimitsMin=<-5 -10 -40>
       angularTarget=<1 0 0>
       linearAxesMax=<0.25 0.25 0.25>
       linearAxesMin=<-0.25 -0.25 -0.25>
      }
      localJointOffset=<3 0 0>
     }]
    planarLimits [{drivingBone {boneName=spine_04}, planeTransform {location=<0 -7.5 0>, rotation=<rot 0 5 -89.9999>}}]
    sphericalLimits [{limitRadius=10.5, drivingBone {boneName=spine_05}, sphereLocalOffset=<0 0 0>}]
   }
   n18 = "Transform (Modify) Bone - Bone: dyn_backBraid_03"(ComponentPose=@n17, Alpha=Get CapeAlphaInv() -> (CapeAlphaInv:double), Translation="0,0,0", Rotation="0,30,0", Scale="1,1,1") {
    rotationMode=BMM_Replace
    rotationSpace=BCS_BoneSpace
    alphaBoolBlend {blendOption=Linear}
    boneToModify {boneName=dyn_backBraid_03}
    rotation {yaw=30}
   }
   n19 = Get templeBraidAlpha() -> (templeBraidAlpha:double)
   n20 = "Transform (Modify) Bone - Bone: dyn_templeBraid_Base_r"(ComponentPose=@n18, Alpha=@n19, Translation="0,0,0", Rotation="-40,0,0", Scale="1,1,1") {
    rotationMode=BMM_Replace
    rotationSpace=BCS_BoneSpace
    alphaBoolBlend {blendOption=Linear}
    boneToModify {boneName=dyn_templeBraid_Base_r}
    rotation {pitch=-40}
   }
   "Transform (Modify) Bone - Bone: dyn_templeBraid_Base_l"(ComponentPose=@n20, Alpha=@n19, Translation="0,0,0", Rotation="-40,0.000,0.000", Scale="1,1,1") {
    rotationMode=BMM_Additive
    rotationSpace=BCS_BoneSpace
    alphaBoolBlend {blendOption=Linear}
    boneToModify {boneName=dyn_templeBraid_Base_l}
    rotation {pitch=-40}
   }
   n21 = RigidBody(
    ComponentPose=Local To Component(LocalPose=Copy Pose From Mesh() {bUseAttachedParent=true})
    ExternalForce=Break Fort Anim Input Skydiving External Force!(FortAnimInput_SkydivingExternalForce=Get SkydivingExternalForceInput()) -> (
     SkydivingVectorForce=~
     HeadToPelvisDirection=~
     FloatingMultiplier=~
     FloatingAdditive=~
     DivingMultiplier=~
     DivingAdditive=~
     ParachutingMultiplier=~
     ParachutingAdditive=~
    )
   ) {
    bClampLinearTranslationLimitToRefPose=true
    bOverrideWorldGravity=true
    overridePhysicsAsset=&/F_MED_Fire_Elf_FaceAcc_Dyn_Braid
    simulationSpace=BaseBoneSpace
    alphaBoolBlend {blendOption=Linear}
    baseBoneRef {boneName=spine_03}
    componentAppliedLinearAccClamp=<2000 2000 2000>
    componentLinearAccScale=<0.45 0.2 0.2>
    componentLinearVelScale=<0.45 0.2 0.2>
    overrideWorldGravity {z=-980}
   }
   n22 = RigidBody(ComponentPose=@n21) {
    bClampLinearTranslationLimitToRefPose=true
    bOverrideWorldGravity=true
    overridePhysicsAsset=&/F_MED_Fire_Elf_Face_TempleBraid_Physics
    simulationSpace=BaseBoneSpace
    alphaBoolBlend {blendOption=Linear}
    baseBoneRef {boneName=spine_03}
    componentAppliedLinearAccClamp=<2000 2000 2000>
    componentLinearAccScale=<0.35 0.2 0.2>
    componentLinearVelScale=<0.35 0.2 0.2>
    overrideWorldGravity {z=-980}
   }
   Output Pose(Result=Component To Local(ComponentPose=@n22)) {layerGroup=None}
   n23 = Flap Node - Bone: dyn_templeBraid_Base_r!(ComponentPose:Component Space Pose Link Structure, Alpha:float=1) -> (Pose:Component Space Pose Link Structure) {
    is=title FortAnimGraphNode_Flap
    bUseClamp=true
    targetClampMax=50
    targetClampMin=-50
    targetComponent=RotationY
    alphaBoolBlend {blendOption=Linear}
    sourceDrives [{
      bUseRange=false
      drivingCurve=null
      multiplier=1
      rangeMax=50
      rangeMin=0
      remappedMax=50
      remappedMin=0
      sourceComponent=RotationZ
      useQuaternion=true
      sourceBone {boneName=head}
     }]
    targetBone {boneName=dyn_templeBraid_Base_r}
   }
   Flap Node - Bone: dyn_templeBraid_Base_l!(ComponentPose:Component Space Pose Link Structure=@n23, Alpha:float=1) -> (Pose:Component Space Pose Link Structure) {
    is=title FortAnimGraphNode_Flap
    bUseClamp=true
    targetClampMin=-1
    targetComponent=RotationY
    alphaBoolBlend {blendOption=Linear}
    sourceDrives [{
      bUseRange=false
      drivingCurve=null
      multiplier=-0.25
      rangeMax=50
      rangeMin=0
      remappedMax=50
      remappedMin=0
      sourceComponent=RotationZ
      useQuaternion=true
      sourceBone {boneName=head}
     }]
    targetBone {boneName=dyn_templeBraid_Base_l}
   }
  ]}
 EventGraph {nodes [
   Event BlueprintInitializeAnimation()
   n1 = TryGetPawnOwner()
   > Is Valid(InputObject=@n1)
   > n2 = Cast To PlayerPawn_Athena(Object=@n1)
   > Set PawnOwner(PawnOwner:Player Pawn Athena Object Reference=@n2) -> (Output_Get:Player Pawn Athena Object Reference)
   Event BlueprintUpdateAnimation()
   n3 = BreakTransform(InTransform=GetDeltaTransformFromRefPose(self=GetOwnerSkeletalMeshComponent(), BoneName=spine_05, BaseName=head))
   > Set templeBraidAlpha(templeBraidAlpha:double=MapRangeClamped(Value=BreakRotator(InRot=@n3.Rotation).Yaw, InRangeA=-28, InRangeB=0, OutRangeA=1, OutRangeB=0)) -> (Output_Get:double)
   Event OnCharacterCustomizationLoaded()
   n4 = CheckForBackbling()
   > Set IsWearingBackbling(IsWearingBackbling:bool=@n4) -> (Output_Get:bool)
   > n5 = Set IsWearingCape(IsWearingCape:bool=@n4.bIsBackblingLong) -> (Output_Get:bool)
   > Branch(Condition=@n5)
   > Do Once()
   > Is Valid(InputObject=Get PawnOwner() -> (PawnOwner:Player Pawn Athena Object Reference))
   n6 = GetSkeletalMeshForPartType(self=Get PawnOwner())
   n7 = "To Float (Boolean)"(InBool="Not Equal (Object)"!(
     A:Object Reference=FindChildCommponentFromAttachSocket!(
      self:Custom Character Part Anim Instance Object Reference
      Parent:Skeletal Mesh Component Object Reference=@n6
      bIncludeAllDescendants:bool=false
      AttachSocket:Name=Cape
     ) -> (ReturnValue:Skeletal Mesh Component Object Reference)
     B:Object Reference
    ) -> (ReturnValue:bool))
   > Set capeAlpha(capeAlpha:double=float - float(B=@n7) {is=title K2Node_CallFunction}) -> (Output_Get:double)
   > Set CapeAlphaInv(CapeAlphaInv:double=@n7) -> (Output_Get:double)
   GetPartSkeletalMeshComponent()
   GetOwnerSkeletalMeshComponent()
   TryGetPawnOwner()
  ]}
]
variables [
 capeAlpha {type=number}
 CapeAlphaInv {type=number}
 IsWearingBackbling {type=boolean}
 IsWearingCape {type=boolean}
 PawnOwner {type=PlayerPawn_Athena_C}
 templeBraidAlpha {type=number}
]
